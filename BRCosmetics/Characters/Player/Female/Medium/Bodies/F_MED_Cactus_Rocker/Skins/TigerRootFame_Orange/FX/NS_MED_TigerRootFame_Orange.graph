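# Niagara effect for the TigerRootFame_Orange skin of F_MED_Cactus_Rocker, drawing audio-reactive glowing mesh rays along the character's arms.
# The emitters shown pin short-lived particles at random offsets along upper and lower arm bones, each with three mesh renderers.
# Audio values gate spawning and drive Z stretch and brightness, while curves cycle color and fade.
# Front-end modifiers retune intensity; a time-of-day value also scales one mesh color.
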NiagaraSystem /BRCosmetics/Characters/Player/Female/Medium/Bodies/F_MED_Cactus_Rocker/Skins/TigerRootFame_Orange/FX/NS_MED_TigerRootFame_Orange
dependencies {
 usedDataInterfaces [
  &NiagaraDataInterfaceSkeletalMesh
  &NiagaraDataInterfaceColorCurve
  &NiagaraDataInterfaceCurve
  &NiagaraDataInterfaceArrayFloat3
 ]
 usedDynamicInputs [&Max_Float]
 usedModules [
  &SystemState
  &self:SystemSpawnScript.SystemScriptSource.SystemScriptGraph.NiagaraNodeAssignment_1.SetVariables_#1
  &self:SystemSpawnScript.SystemScriptSource.SystemScriptGraph.NiagaraNodeAssignment_0.SetVariables_#2
 ]
}
emitters {
 Blocks_Right_Ray {
  emitterUpdateScript [
   &EmitterState(
    Inactive Response=::NewEnumerator1
    Life Cycle Mode=::NewEnumerator1
    Loop Behavior=::NewEnumerator1
    Loop Duration Mode=::NewEnumerator1
    UseLoopDelay=false
   )
   &SpawnRate(
    SpawnRate=&MakeCustomFloatFromBool(Bool=&SetBoolByFloatComparison(A=$System.Audio NPC, B=0), True Float=70, False Float=0)
    Use Spawn Probability=false
   )
  ]
  particleSpawnScript [
   &FN_SetSpawnGroup(Number of Spawn Groups=5)
   &InitializeParticle(
    Color=<rgba 1.5 1.5 1.5 3>
    Lifetime=0.2
    Mesh Scale=&Multiply_VectorByFloat(Vector=&RandomRangeVector(Minimum=<0.5 0.5 0.5>, Maximum=<0.5 0.5 0.8>), Float=0.8)
    Mesh Scale Mode=::NewEnumerator4
    Sprite Size Mode=::NewEnumerator3
    UsePositionOffset=false
    Write Material Random=true
   ) {module=InitializeParticle001}
   &SphereLocation(
    MaskSpawnGroup=false
    Override Seed=false
    Sphere Origin:NiagaraPosition=&ConvertVectorToPosition(Input Position=<0 0 0>, Vector Is In Absolute Large World Space:NiagaraBool=false)
    Sphere Radius=1
    Surface Only=false
    UseNonUniformScale=false
    UseOffset=false
    UseSphereOrientation=false
   )
   /SetVariables(Particles.Random=&RandomRangeFloat(Minimum=0, Maximum=1)) {isSetParameters=true, module=SetVariables_#3}
   &InitialMeshOrientation(Mesh Orientation Mode=::NewEnumerator2, Rotation=<0 0 0.1>, Use Rotation Vector=true)
  ]
  particleUpdateScript [
   &ParticleState()
   &SolveForcesAndVelocity(Clamp Velocity=false, Limit Acceleration=false)
   &SkeletalMeshLocation(
    Bone Sampling Mode=::NewEnumerator4
    MaskSpawnGroup=false
    Orientation Sampling=::NewEnumerator1
    Sampled Position Offset=&Add_Vector(
     A=&ConvertPositionToVector(Input Position=$Particles.Initial.Position, Convert To Absolute Large World Vector=false)
     B=&Lerp_Vector(A=<5 0 -4>, B=<-23 0 -4>, Alpha=$Particles.Random)
    )
    Skeletal Mesh=DataInterface() {propertyValues {
      DefaultMesh=null
      ExcludeBoneName=root
      OptionalComponentClass=null
      PreviewMesh=/BRCosmetics/Characters/Player/Female/Medium/Bodies/F_MED_Cactus_Rocker/Meshes/F_MED_Cactus_Rocker
      SkinningMode=SkinOnTheFly
      SoftSourceActor=null
      SourceMode=Default
      UvSetIndex=0
      WholeMeshLOD=-1
      bExcludeBone=false
      bReadDeformedGeometry=true
      bRequireCurrentFrameData=true
      ComponentTags []
      FilteredBones [upperarm_r]
      FilteredSockets []
      MeshUserParameter {parameter {name=None, typeDefHandle {registeredTypeIndex=1}, varData []}}
      SamplingRegions []
     }}
   )
   &ScaleColor(
    Scale Alpha=&Multiply_Float(
     A=&FloatFromCurve(
      FloatCurve=DataInterface() {propertyValues {Curve {keys [
          {interpMode=RCIM_Cubic, tangentMode=RCTM_User, time=0, value=0}
          {interpMode=RCIM_Cubic, tangentMode=RCTM_User, time=0.5, value=1}
          {tangentMode=RCTM_User, time=1, value=0}
         ]}}}
      Scale Curve=&Multiply_Float(
       A=&FN_FrontEndFloatModifier(
        Gameplay=&Multiply_Float(
         A=&FloatFromCurve(
          FloatCurve=DataInterface() {propertyValues {Curve {keys [
              {interpMode=RCIM_Cubic, tangentMode=RCTM_User, time=0, value=0}
              {interpMode=RCIM_Cubic, tangentMode=RCTM_User, time=0.3, value=1}
              <at 0.7 0.9>
              {tangentMode=RCTM_User, time=1, value=0}
             ]}}}
          CurveIndex=$Particles.Random
         )
         B=1
        )
        Front End=1
       )
       B=1
      )
     )
     B=&FN_FrontEndFloatModifier(Gameplay=1.25, Front End=0.75)
    )
    Scale RGB=&Multiply_VectorByFloat(
     Vector=&MakeVectorFromLinearColorRGB(Color=&ColorFromCurve(
       ColorCurve=DataInterface() {propertyValues {
         AlphaCurve {keys [<at 0 1 RCIM_Cubic>]}
         BlueCurve {keys [<at 0.2424 0.0417>,<at 0.2989 1>,<at 0.4745 1>,<at 0.6883 1>,<at 0.7448 0.0417 RCIM_Cubic>]}
         GreenCurve {keys [<at 0.2424 0.3611>,<at 0.2989 0.0833>,<at 0.4745 0.8333>,<at 0.6883 0>,<at 0.7448 0.3611 RCIM_Cubic>]}
         RedCurve {keys [<at 0.2424 1>,<at 0.2989 1>,<at 0.4745 0>,<at 0.6883 0.1667>,<at 0.7448 1 RCIM_Cubic>]}
        }}
       CurveIndex=&Multiply_Float(A=$Particles.Random, B=$Particles.NormalizedAge)
      ))
     Float=&Multiply_Float(
      A=&ClampFloat(Float=$System.Audio NPC, Min=1)
      B=&FN_FrontEndFloatModifier(Gameplay=1, Front End=0.75)
     )
    )
    ScaleA=true
    ScaleRGB=true
   ) {module=ScaleColor002}
   &ScaleMeshSize(Scale Factor=&MakeVector(
     X=1
     Y=1
     Z=&FloatFromCurve(
      FloatCurve=DataInterface() {propertyValues {Curve {keys [
          {interpMode=RCIM_Cubic, tangentMode=RCTM_User, time=0, value=0}
          {interpMode=RCIM_Cubic, tangentMode=RCTM_User, time=0.5, value=1}
          {tangentMode=RCTM_User, time=1, value=0}
         ]}}}
      Scale Curve=&Lerp_Float(
       A=1
       B=&FN_FrontEndFloatModifier(Gameplay=2.5, Front End=2.5)
       Alpha=&ClampFloat(Float=$System.Audio NPC)
      )
     )
    )) {module=ScaleMeshSize001}
   /SetVariables(Particles.Glow Color=&MakeLinearColorFromVectorAndFloat(
     "Vector (RGB)"=&Multiply_VectorByFloat(
      Vector=&Multiply_VectorByFloat(
       Vector=&MakeVectorFromLinearColorRGB(Color=&ColorFromCurve(
         ColorCurve=DataInterface() {propertyValues {
           AlphaCurve {keys [<at 0 1 RCIM_Cubic>]}
           BlueCurve {keys [
             <at 0.0075 0>
             <at 0.0927 0>
             <at 0.2424 0.0417>
             <at 0.2989 1>
             <at 0.4733 1>
             <at 0.6883 1>
             {arriveTangent=-6.0278, interpMode=RCIM_Cubic, leaveTangent=-6.0278, time=0.7448, value=0.0417}
             <at 0.8542 0>
             <at 0.9827 0>
            ]}
           GreenCurve {keys [
             <at 0.0075 0.6667>
             <at 0.0927 0.1667>
             <at 0.2424 0.5208>
             <at 0.2989 0.0833>
             <at 0.4733 0.1667>
             <at 0.6883 0>
             {arriveTangent=1.0046, interpMode=RCIM_Cubic, leaveTangent=1.0046, time=0.7448, value=0.5208}
             <at 0.8542 0.1667>
             <at 0.9827 0.6667>
            ]}
           RedCurve {keys [
             <at 0.0075 1>
             <at 0.0927 1>
             <at 0.2424 1>
             <at 0.2989 1>
             <at 0.4733 0>
             <at 0.6883 0.1667>
             {arriveTangent=5.0231, interpMode=RCIM_Cubic, leaveTangent=5.0231, time=0.7448, value=1}
             <at 0.8542 1>
             <at 0.9827 1>
            ]}
          }}
         CurveIndex=&Multiply_Float(A=$Particles.Random, B=$Particles.NormalizedAge)
        ))
       Float=&Multiply_Float(
        A=&ClampFloat(Float=$System.Audio NPC, Min=1)
        B=&FN_FrontEndFloatModifier(Gameplay=1, Front End=0.85)
       )
      )
      Float=0.6
     )
     "Float (Alpha)"=&Multiply_Float(
      A=&Multiply_Float(
       A=&FloatFromCurve(
        FloatCurve=DataInterface() {propertyValues {Curve {keys [
            {interpMode=RCIM_Cubic, tangentMode=RCTM_User, time=0, value=0}
            <at 0.2 1>
            {interpMode=RCIM_Cubic, tangentMode=RCTM_User, time=0.8, value=1}
            {tangentMode=RCTM_User, time=1, value=0}
           ]}}}
        Scale Curve=&Multiply_Float(
         A=&FN_FrontEndFloatModifier(Gameplay=&Multiply_Float(A=1, B=1), Front End=1)
         B=&RandomRangeFloat(Minimum=1, Maximum=1)
        )
       )
       B=1
      )
      B=&FN_FrontEndFloatModifier(Gameplay=2, Front End=3)
     )
    )) {isSetParameters=true, module=SetVariables_#4}
   /SetVariables(Particles.Glow Size:Vector3f=&Multiply_Vector(
     A=&Multiply_VectorByFloat(
      Vector=&RandomRangeVector(Minimum=<0.5 0.5 0.8>, Maximum=<0.6 0.6 1.3>)
      Float=&RandomRangeFloat(Minimum=1, Maximum=1)
     )
     B=&MakeVector(
      X=1
      Y=1
      Z=&FloatFromCurve(
       FloatCurve=DataInterface() {propertyValues {Curve {keys [
           {interpMode=RCIM_Cubic, tangentMode=RCTM_User, time=0, value=0}
           {interpMode=RCIM_Cubic, tangentMode=RCTM_User, time=0.5, value=1}
           {tangentMode=RCTM_User, time=1, value=0}
          ]}}}
       Scale Curve=&Lerp_Float(
        A=1
        B=&FN_FrontEndFloatModifier(Gameplay=3.2, Front End=3.2)
        Alpha=&ClampFloat(Float=$System.Audio NPC)
       )
      )
     )
    )) {isSetParameters=true, module=SetVariables_#5}
   /SetVariables(Particles.Mesh size02:Vector3f=&Multiply_Vector(
     A=&Multiply_VectorByFloat(
      Vector=&RandomRangeVector(Minimum=<0.1 0.1 1>, Maximum=<0.1 0.1 1>)
      Float=&RandomRangeFloat(Minimum=1, Maximum=1)
     )
     B=&MakeVector(
      X=1
      Y=1
      Z=&FloatFromCurve(
       FloatCurve=DataInterface() {propertyValues {Curve {keys [
           {interpMode=RCIM_Cubic, tangentMode=RCTM_User, time=0, value=0}
           {interpMode=RCIM_Cubic, tangentMode=RCTM_User, time=0.5, value=1}
           {tangentMode=RCTM_User, time=1, value=0}
          ]}}}
       Scale Curve=&Lerp_Float(
        A=2
        B=&FN_FrontEndFloatModifier(Gameplay=2.5, Front End=2.2)
        Alpha=&ClampFloat(Float=$NPC.FortniteNPC.EventAudio_Amplitude)
       )
      )
     )
    )) {isSetParameters=true}
   /SetVariables(Particles.Mesh02 Color:LinearColor=&MakeLinearColorFromVectorAndFloat(
     "Vector (RGB)"=&Multiply_VectorByFloat(
      Vector=&MakeVectorFromLinearColorRGB(Color=&ColorFromCurve(
        ColorCurve=DataInterface() {propertyValues {
          AlphaCurve {keys [<at 0 1 RCIM_Cubic>]}
          BlueCurve {keys [
            <at 0.0075 0>
            <at 0.2424 0.0417>
            <at 0.2989 1>
            <at 0.4733 1>
            <at 0.6883 1>
            {arriveTangent=-3.3972, interpMode=RCIM_Cubic, leaveTangent=-3.3972, time=0.7448, value=0.0417}
            <at 0.9827 0>
           ]}
          GreenCurve {keys [
            <at 0.0075 0.6667>
            <at 0.2424 0.5208>
            <at 0.2989 0.0833>
            <at 0.4733 0.1667>
            <at 0.6883 0>
            {arriveTangent=2.2648, interpMode=RCIM_Cubic, leaveTangent=2.2648, time=0.7448, value=0.5208}
            <at 0.9827 0.6667>
           ]}
          RedCurve {keys [
            <at 0.0075 1>
            <at 0.2424 1>
            <at 0.2989 1>
            <at 0.4733 0>
            <at 0.6883 0.1667>
            {arriveTangent=2.831, interpMode=RCIM_Cubic, leaveTangent=2.831, time=0.7448, value=1}
            <at 0.9827 1>
           ]}
         }}
        CurveIndex=&Multiply_Float(A=$Particles.Random, B=$Particles.NormalizedAge)
       ))
      Float=&Multiply_Float(A=&FN_FrontEndFloatModifier(Gameplay=3, Front End=1.5), B=$System.TOD)
     )
     "Float (Alpha)"=&FloatFromCurve(
      FloatCurve=DataInterface() {propertyValues {Curve {keys [
          {interpMode=RCIM_Cubic, tangentMode=RCTM_User, time=0, value=0}
          {interpMode=RCIM_Cubic, tangentMode=RCTM_User, time=0.5, value=1}
          {tangentMode=RCTM_User, time=1, value=0}
         ]}}}
      Scale Curve=&Multiply_Float(A=&FN_FrontEndFloatModifier(Gameplay=1, Front End=2), B=4)
     )
    )) {isSetParameters=true}
  ]
  rendererClasses [&NiagaraMeshRendererProperties]
  renderers [
   {rendererClass=&NiagaraMeshRendererProperties, rendererIndex=0}
   {rendererClass=&NiagaraMeshRendererProperties, rendererIndex=1}
   {rendererClass=&NiagaraMeshRendererProperties, rendererIndex=2}
  ]
 }
 Blocks_left_Ray {
  emitterUpdateScript [
   &EmitterState(
    Inactive Response=::NewEnumerator1
    Life Cycle Mode=::NewEnumerator1
    Loop Behavior=::NewEnumerator1
    Loop Duration Mode=::NewEnumerator1
    UseLoopDelay=false
   )
   &SpawnRate(
    SpawnRate=&MakeCustomFloatFromBool(Bool=&SetBoolByFloatComparison(A=$System.Audio NPC, B=0), True Float=70, False Float=0)
    Use Spawn Probability=false
   )
  ]
  particleSpawnScript [
   &FN_SetSpawnGroup(Number of Spawn Groups=3)
   &InitializeParticle(
    Color=<rgba 1.5 1.5 1.5 3>
    Lifetime=0.2
    Mesh Scale=&Multiply_VectorByFloat(Vector=&RandomRangeVector(Minimum=<0.5 0.5 0.5>, Maximum=<0.5 0.5 0.8>), Float=0.8)
    Mesh Scale Mode=::NewEnumerator4
    Sprite Size Mode=::NewEnumerator3
    UsePositionOffset=false
    Write Material Random=true
   )
   &SphereLocation(
    MaskSpawnGroup=false
    Override Seed=false
    Sphere Origin:NiagaraPosition=&ConvertVectorToPosition(Input Position=<0 0 0>, Vector Is In Absolute Large World Space:NiagaraBool=false)
    Sphere Radius=1
    Surface Only=false
    UseNonUniformScale=false
    UseOffset=false
    UseSphereOrientation=false
   )
   /SetVariables(Particles.Random=&RandomRangeFloat(Minimum=0, Maximum=1)) {isSetParameters=true, module=SetVariables_#3}
   &InitialMeshOrientation(Mesh Orientation Mode=::NewEnumerator2, Rotation=<0 0 0.1>, Use Rotation Vector=true)
  ]
  particleUpdateScript [
   &ParticleState()
   &SolveForcesAndVelocity(Clamp Velocity=false, Limit Acceleration=false)
   &SkeletalMeshLocation(
    Bone Sampling Mode=::NewEnumerator4
    MaskSpawnGroup=false
    Orientation Sampling=::NewEnumerator1
    Sampled Position Offset=&Add_Vector(
     A=&ConvertPositionToVector(Input Position=$Particles.Initial.Position, Convert To Absolute Large World Vector=false)
     B=&LerpPosition(A=<-2 0 4>, B=<25 0 4>, Alpha=$Particles.Random)
    )
    Skeletal Mesh=DataInterface() {propertyValues {
      DefaultMesh=null
      ExcludeBoneName=root
      OptionalComponentClass=null
      PreviewMesh=&F_MED_Cactus_Rocker
      SkinningMode=SkinOnTheFly
      SoftSourceActor=null
      SourceMode=Default
      UvSetIndex=0
      WholeMeshLOD=-1
      bExcludeBone=false
      bReadDeformedGeometry=true
      bRequireCurrentFrameData=true
      ComponentTags []
      FilteredBones [upperarm_l]
      FilteredSockets []
      MeshUserParameter {parameter {name=None, typeDefHandle {registeredTypeIndex=1}, varData []}}
      SamplingRegions []
     }}
   )
   &ScaleColor(
    Scale Alpha=&Multiply_Float(
     A=&FloatFromCurve(
      FloatCurve=DataInterface() {propertyValues {Curve {keys [
          {interpMode=RCIM_Cubic, tangentMode=RCTM_User, time=0, value=0}
          {interpMode=RCIM_Cubic, tangentMode=RCTM_User, time=0.5, value=1}
          {tangentMode=RCTM_User, time=1, value=0}
         ]}}}
      Scale Curve=&Multiply_Float(
       A=&FN_FrontEndFloatModifier(
        Gameplay=&Multiply_Float(
         A=&FloatFromCurve(
          FloatCurve=DataInterface() {propertyValues {Curve {keys [
              {interpMode=RCIM_Cubic, tangentMode=RCTM_User, time=0, value=0}
              {interpMode=RCIM_Cubic, tangentMode=RCTM_User, time=0.3, value=1}
              <at 0.7 0.9>
              {tangentMode=RCTM_User, time=1, value=0}
             ]}}}
          CurveIndex=$Particles.Random
         )
         B=1
        )
        Front End=1
       )
       B=1
      )
     )
     B=&FN_FrontEndFloatModifier(Gameplay=1.25, Front End=0.75)
    )
    Scale RGB=&Multiply_VectorByFloat(
     Vector=&MakeVectorFromLinearColorRGB(Color=&ColorFromCurve(
       ColorCurve=DataInterface() {propertyValues {
         AlphaCurve {keys [<at 0 1 RCIM_Cubic>]}
         BlueCurve {keys [<at 0.2424 0.0417>,<at 0.2989 1>,<at 0.4829 1>,<at 0.6883 1>,<at 0.7448 0.0417 RCIM_Cubic>]}
         GreenCurve {keys [<at 0.2424 0.3611>,<at 0.2989 0.0833>,<at 0.4829 0.8333>,<at 0.6883 0>,<at 0.7448 0.3611 RCIM_Cubic>]}
         RedCurve {keys [<at 0.2424 1>,<at 0.2989 1>,<at 0.4829 0>,<at 0.6883 0.1667>,<at 0.7448 1 RCIM_Cubic>]}
        }}
       CurveIndex=&Multiply_Float(A=$Particles.Random, B=$Particles.NormalizedAge)
      ))
     Float=&Multiply_Float(
      A=&ClampFloat(Float=$System.Audio NPC, Min=1)
      B=&FN_FrontEndFloatModifier(Gameplay=1, Front End=0.75)
     )
    )
    ScaleA=true
    ScaleRGB=true
   )
   &ScaleMeshSize(Scale Factor=&MakeVector(
     X=1
     Y=1
     Z=&FloatFromCurve(
      FloatCurve=DataInterface() {propertyValues {Curve {keys [
          {interpMode=RCIM_Cubic, tangentMode=RCTM_User, time=0, value=0}
          {interpMode=RCIM_Cubic, tangentMode=RCTM_User, time=0.5, value=1}
          {tangentMode=RCTM_User, time=1, value=0}
         ]}}}
      Scale Curve=&Lerp_Float(
       A=1
       B=&FN_FrontEndFloatModifier(Gameplay=2.5, Front End=2.5)
       Alpha=&ClampFloat(Float=$System.Audio NPC)
      )
     )
    ))
   /SetVariables(Particles.Glow Color=&MakeLinearColorFromVectorAndFloat(
     "Vector (RGB)"=&Multiply_VectorByFloat(
      Vector=&Multiply_VectorByFloat(
       Vector=&MakeVectorFromLinearColorRGB(Color=&ColorFromCurve(
         ColorCurve=DataInterface() {propertyValues {
           AlphaCurve {keys [<at 0 1 RCIM_Cubic>]}
           BlueCurve {keys [
             <at 0.0075 0>
             <at 0.0927 0>
             <at 0.2424 0.0417>
             <at 0.2989 1>
             <at 0.4733 1>
             <at 0.6883 1>
             {arriveTangent=-6.0278, interpMode=RCIM_Cubic, leaveTangent=-6.0278, time=0.7448, value=0.0417}
             <at 0.8542 0>
             <at 0.9827 0>
            ]}
           GreenCurve {keys [
             <at 0.0075 0.6667>
             <at 0.0927 0.1667>
             <at 0.2424 0.5208>
             <at 0.2989 0.0833>
             <at 0.4733 0.1667>
             <at 0.6883 0>
             {arriveTangent=1.0046, interpMode=RCIM_Cubic, leaveTangent=1.0046, time=0.7448, value=0.5208}
             <at 0.8542 0.1667>
             <at 0.9827 0.6667>
            ]}
           RedCurve {keys [
             <at 0.0075 1>
             <at 0.0927 1>
             <at 0.2424 1>
             <at 0.2989 1>
             <at 0.4733 0>
             <at 0.6883 0.1667>
             {arriveTangent=5.0231, interpMode=RCIM_Cubic, leaveTangent=5.0231, time=0.7448, value=1}
             <at 0.8542 1>
             <at 0.9827 1>
            ]}
          }}
         CurveIndex=&Multiply_Float(A=$Particles.Random, B=$Particles.NormalizedAge)
        ))
       Float=&Multiply_Float(
        A=&ClampFloat(Float=$System.Audio NPC, Min=1)
        B=&FN_FrontEndFloatModifier(Gameplay=1, Front End=0.85)
       )
      )
      Float=0.6
     )
     "Float (Alpha)"=&Multiply_Float(
      A=&Multiply_Float(
       A=&FloatFromCurve(
        FloatCurve=DataInterface() {propertyValues {Curve {keys [
            {interpMode=RCIM_Cubic, tangentMode=RCTM_User, time=0, value=0}
            <at 0.2 1>
            {interpMode=RCIM_Cubic, tangentMode=RCTM_User, time=0.8, value=1}
            {tangentMode=RCTM_User, time=1, value=0}
           ]}}}
        Scale Curve=&Multiply_Float(
         A=&FN_FrontEndFloatModifier(Gameplay=&Multiply_Float(A=1, B=1), Front End=1)
         B=&RandomRangeFloat(Minimum=1, Maximum=1)
        )
       )
       B=1
      )
      B=&FN_FrontEndFloatModifier(Gameplay=2, Front End=3)
     )
    )) {isSetParameters=true, module=SetVariables_#4}
   /SetVariables(Particles.Glow Size:Vector3f=&Multiply_Vector(
     A=&Multiply_VectorByFloat(
      Vector=&RandomRangeVector(Minimum=<0.5 0.5 0.8>, Maximum=<0.6 0.6 1.3>)
      Float=&RandomRangeFloat(Minimum=1, Maximum=1)
     )
     B=&MakeVector(
      X=1
      Y=1
      Z=&FloatFromCurve(
       FloatCurve=DataInterface() {propertyValues {Curve {keys [
           {interpMode=RCIM_Cubic, tangentMode=RCTM_User, time=0, value=0}
           {interpMode=RCIM_Cubic, tangentMode=RCTM_User, time=0.5, value=1}
           {tangentMode=RCTM_User, time=1, value=0}
          ]}}}
       Scale Curve=&Lerp_Float(
        A=1
        B=&FN_FrontEndFloatModifier(Gameplay=3.2, Front End=3.2)
        Alpha=&ClampFloat(Float=$System.Audio NPC)
       )
      )
     )
    )) {isSetParameters=true, module=SetVariables_#5}
   /SetVariables(Particles.Mesh size02:Vector3f=&Multiply_Vector(
     A=&Multiply_VectorByFloat(
      Vector=&RandomRangeVector(Minimum=<0.1 0.1 1>, Maximum=<0.1 0.1 1>)
      Float=&RandomRangeFloat(Minimum=1, Maximum=1)
     )
     B=&MakeVector(
      X=1
      Y=1
      Z=&FloatFromCurve(
       FloatCurve=DataInterface() {propertyValues {Curve {keys [
           {interpMode=RCIM_Cubic, tangentMode=RCTM_User, time=0, value=0}
           {interpMode=RCIM_Cubic, tangentMode=RCTM_User, time=0.5, value=1}
           {tangentMode=RCTM_User, time=1, value=0}
          ]}}}
       Scale Curve=&Lerp_Float(
        A=2
        B=&FN_FrontEndFloatModifier(Gameplay=2.5, Front End=2.2)
        Alpha=&ClampFloat(Float=$NPC.FortniteNPC.EventAudio_Amplitude)
       )
      )
     )
    )) {isSetParameters=true}
   /SetVariables(Particles.Mesh02 Color:LinearColor=&MakeLinearColorFromVectorAndFloat(
     "Vector (RGB)"=&Multiply_VectorByFloat(
      Vector=&MakeVectorFromLinearColorRGB(Color=&ColorFromCurve(
        ColorCurve=DataInterface() {propertyValues {
          AlphaCurve {keys [<at 0 1 RCIM_Cubic>]}
          BlueCurve {keys [
            <at 0.0075 0>
            <at 0.2424 0.0417>
            <at 0.2989 1>
            <at 0.4733 1>
            <at 0.6883 1>
            {arriveTangent=-3.3972, interpMode=RCIM_Cubic, leaveTangent=-3.3972, time=0.7448, value=0.0417}
            <at 0.9827 0>
           ]}
          GreenCurve {keys [
            <at 0.0075 0.6667>
            <at 0.2424 0.5208>
            <at 0.2989 0.0833>
            <at 0.4733 0.1667>
            <at 0.6883 0>
            {arriveTangent=2.2648, interpMode=RCIM_Cubic, leaveTangent=2.2648, time=0.7448, value=0.5208}
            <at 0.9827 0.6667>
           ]}
          RedCurve {keys [
            <at 0.0075 1>
            <at 0.2424 1>
            <at 0.2989 1>
            <at 0.4733 0>
            <at 0.6883 0.1667>
            {arriveTangent=2.831, interpMode=RCIM_Cubic, leaveTangent=2.831, time=0.7448, value=1}
            <at 0.9827 1>
           ]}
         }}
        CurveIndex=&Multiply_Float(A=$Particles.Random, B=$Particles.NormalizedAge)
       ))
      Float=&Multiply_Float(A=&FN_FrontEndFloatModifier(Gameplay=3, Front End=1.5), B=$System.TOD)
     )
     "Float (Alpha)"=&FloatFromCurve(
      FloatCurve=DataInterface() {propertyValues {Curve {keys [
          {interpMode=RCIM_Cubic, tangentMode=RCTM_User, time=0, value=0}
          {interpMode=RCIM_Cubic, tangentMode=RCTM_User, time=0.5, value=1}
          {tangentMode=RCTM_User, time=1, value=0}
         ]}}}
      Scale Curve=&Multiply_Float(A=&FN_FrontEndFloatModifier(Gameplay=1, Front End=2), B=4)
     )
    )) {isSetParameters=true}
  ]
  rendererClasses [&NiagaraMeshRendererProperties]
  renderers [
   {rendererClass=&NiagaraMeshRendererProperties, rendererIndex=0}
   {rendererClass=&NiagaraMeshRendererProperties, rendererIndex=1}
   {rendererClass=&NiagaraMeshRendererProperties, rendererIndex=2}
  ]
 }
 Blocks_left_Ray001 {
  emitterUpdateScript [
   &EmitterState(
    Inactive Response=::NewEnumerator1
    Life Cycle Mode=::NewEnumerator1
    Loop Behavior=::NewEnumerator1
    Loop Duration Mode=::NewEnumerator1
    UseLoopDelay=false
   )
   &SpawnRate(
    SpawnRate=&MakeCustomFloatFromBool(Bool=&SetBoolByFloatComparison(A=$System.Audio NPC, B=0), True Float=50, False Float=0)
    Use Spawn Probability=false
   )
  ]
  particleSpawnScript [
   &FN_SetSpawnGroup(Number of Spawn Groups=3)
   &InitializeParticle(
    Color=<rgba 1.2 1.2 1.2 2>
    Lifetime=0.2
    Mesh Scale=&Multiply_VectorByFloat(
     Vector=&RandomRangeVector(Minimum=<0.2 0.2 0.4>, Maximum=<0.3 0.3 0.8>)
     Float=&RandomRangeFloat(Minimum=1, Maximum=1)
    )
    Mesh Scale Mode=::NewEnumerator4
    Sprite Size Mode=::NewEnumerator3
    UsePositionOffset=false
    Write Material Random=true
   )
   &SphereLocation(
    MaskSpawnGroup=false
    Override Seed=false
    Sphere Origin:NiagaraPosition=&ConvertVectorToPosition(Input Position=<0 0 0>, Vector Is In Absolute Large World Space:NiagaraBool=false)
    Sphere Radius=1
    Surface Only=false
    UseNonUniformScale=false
    UseOffset=false
    UseSphereOrientation=false
   )
   /SetVariables(Particles.Random=&RandomRangeFloat(Minimum=0, Maximum=1)) {isSetParameters=true, module=SetVariables_#3}
   &InitialMeshOrientation(Mesh Orientation Mode=::NewEnumerator2, Rotation=<0 0 0.25>, Use Rotation Vector=true)
  ]
  particleUpdateScript [
   &ParticleState()
   &SolveForcesAndVelocity(Clamp Velocity=false, Limit Acceleration=false)
   &SkeletalMeshLocation(
    Bone Sampling Mode=::NewEnumerator4
    MaskSpawnGroup=false
    Orientation Sampling=::NewEnumerator1
    Sampled Position Offset=&Add_Vector(
     A=&ConvertPositionToVector(Input Position=$Particles.Initial.Position, Convert To Absolute Large World Vector=false)
     B=&LerpPosition(A=<0 0 2>, B=<23 0 2>, Alpha=$Particles.Random)
    )
    Skeletal Mesh=DataInterface() {propertyValues {
      DefaultMesh=null
      ExcludeBoneName=root
      OptionalComponentClass=null
      PreviewMesh=&F_MED_Cactus_Rocker
      SkinningMode=SkinOnTheFly
      SoftSourceActor=null
      SourceMode=Default
      UvSetIndex=0
      WholeMeshLOD=-1
      bExcludeBone=false
      bReadDeformedGeometry=true
      bRequireCurrentFrameData=true
      ComponentTags []
      FilteredBones [lowerarm_l]
      FilteredSockets []
      MeshUserParameter {parameter {name=None, typeDefHandle {registeredTypeIndex=1}, varData []}}
      SamplingRegions []
     }}
   )
   &ScaleColor(
    Scale Alpha=&Multiply_Float(
     A=&FloatFromCurve(
      FloatCurve=DataInterface() {propertyValues {Curve {keys [
          {interpMode=RCIM_Cubic, tangentMode=RCTM_User, time=0, value=0}
          {interpMode=RCIM_Cubic, tangentMode=RCTM_User, time=0.5, value=1}
          {tangentMode=RCTM_User, time=1, value=0}
         ]}}}
      Scale Curve=&Multiply_Float(
       A=&FN_FrontEndFloatModifier(Gameplay=&Multiply_Float(A=1, B=1), Front End=1)
       B=2
      )
     )
     B=&FN_FrontEndFloatModifier(Gameplay=1.25, Front End=0.75)
    )
    Scale RGB=&Multiply_VectorByFloat(
     Vector=&MakeVectorFromLinearColorRGB(Color=&ColorFromCurve(
       ColorCurve=DataInterface() {propertyValues {
         AlphaCurve {keys [<at 0 1 RCIM_Cubic>]}
         BlueCurve {keys [
           <at 0 0>
           <at 0.2424 0.0417>
           <at 0.2989 1>
           <at 0.6883 1>
           {arriveTangent=-3.2086, interpMode=RCIM_Cubic, leaveTangent=-3.2086, time=0.7448, value=0.0417}
           <at 1 0>
          ]}
         GreenCurve {keys [
           <at 0 0.6667>
           <at 0.2424 0.8403>
           <at 0.2989 0.0833>
           <at 0.6883 0>
           {arriveTangent=2.1391, interpMode=RCIM_Cubic, leaveTangent=2.1391, time=0.7448, value=0.8403}
           <at 1 0.6667>
          ]}
         RedCurve {keys [
           <at 0 1>
           <at 0.2424 1>
           <at 0.2989 0.5417>
           <at 0.6883 0>
           {arriveTangent=3.2086, interpMode=RCIM_Cubic, leaveTangent=3.2086, time=0.7448, value=1}
           <at 1 1>
          ]}
        }}
       CurveIndex=&Multiply_Float(A=$Particles.Random, B=$Particles.NormalizedAge)
      ))
     Float=&Multiply_Float(
      A=&ClampFloat(Float=$System.Audio NPC, Min=1)
      B=&FN_FrontEndFloatModifier(Gameplay=1, Front End=0.75)
     )
    )
    ScaleA=true
    ScaleRGB=true
   )
   &ScaleMeshSize(Scale Factor=&MakeVector(
     X=1
     Y=1
     Z=&FloatFromCurve(
      FloatCurve=DataInterface() {propertyValues {Curve {keys [
          {interpMode=RCIM_Cubic, tangentMode=RCTM_User, time=0, value=0}
          {interpMode=RCIM_Cubic, tangentMode=RCTM_User, time=0.5, value=1}
          {tangentMode=RCTM_User, time=1, value=0}
         ]}}}
      Scale Curve=&Lerp_Float(
       A=1
       B=&FN_FrontEndFloatModifier(Gameplay=1.5, Front End=1.5)
       Alpha=&ClampFloat(Float=$System.Audio NPC)
      )
     )
    ))
   /SetVariables(Particles.Glow Color=&MakeLinearColorFromVectorAndFloat(
     "Vector (RGB)"=&Multiply_VectorByFloat(
      Vector=&Multiply_VectorByFloat(
       Vector=&MakeVectorFromLinearColorRGB(Color=&ColorFromCurve(
         ColorCurve=DataInterface() {propertyValues {
           AlphaCurve {keys [<at 0 1 RCIM_Cubic>]}
           BlueCurve {keys [
             <at 0.0075 0>
             <at 0.0927 0>
             <at 0.2424 0.0417>
             <at 0.2989 1>
             <at 0.4733 1>
             <at 0.6883 1>
             {arriveTangent=-6.0278, interpMode=RCIM_Cubic, leaveTangent=-6.0278, time=0.7448, value=0.0417}
             <at 0.8542 0>
             <at 0.9827 0>
            ]}
           GreenCurve {keys [
             <at 0.0075 0.6667>
             <at 0.0927 0.1667>
             <at 0.2424 0.5208>
             <at 0.2989 0.0833>
             <at 0.4733 0.1667>
             <at 0.6883 0>
             {arriveTangent=1.0046, interpMode=RCIM_Cubic, leaveTangent=1.0046, time=0.7448, value=0.5208}
             <at 0.8542 0.1667>
             <at 0.9827 0.6667>
            ]}
           RedCurve {keys [
             <at 0.0075 1>
             <at 0.0927 1>
             <at 0.2424 1>
             <at 0.2989 1>
             <at 0.4733 0>
             <at 0.6883 0.1667>
             {arriveTangent=5.0231, interpMode=RCIM_Cubic, leaveTangent=5.0231, time=0.7448, value=1}
             <at 0.8542 1>
             <at 0.9827 1>
            ]}
          }}
         CurveIndex=&Multiply_Float(A=$Particles.Random, B=$Particles.NormalizedAge)
        ))
       Float=&Multiply_Float(
        A=&ClampFloat(Float=$System.Audio NPC, Min=1)
        B=&FN_FrontEndFloatModifier(Gameplay=1, Front End=0.85)
       )
      )
      Float=0.6
     )
     "Float (Alpha)"=&Multiply_Float(
      A=&Multiply_Float(
       A=&FloatFromCurve(
        FloatCurve=DataInterface() {propertyValues {Curve {keys [
            {interpMode=RCIM_Cubic, tangentMode=RCTM_User, time=0, value=0}
            <at 0.2 1>
            {interpMode=RCIM_Cubic, tangentMode=RCTM_User, time=0.8, value=1}
            {tangentMode=RCTM_User, time=1, value=0}
           ]}}}
        Scale Curve=&Multiply_Float(
         A=&FN_FrontEndFloatModifier(Gameplay=&Multiply_Float(A=1, B=1), Front End=1)
         B=&RandomRangeFloat(Minimum=1, Maximum=1)
        )
       )
       B=2
      )
      B=&FN_FrontEndFloatModifier(Gameplay=2, Front End=3)
     )
    )) {isSetParameters=true, module=SetVariables_#4}
   /SetVariables(Particles.Glow Size:Vector3f=&Multiply_Vector(
     A=&Multiply_VectorByFloat(
      Vector=&RandomRangeVector(Minimum=<0.3 0.3 0.7>, Maximum=<0.5 0.5 1.2>)
      Float=&RandomRangeFloat(Minimum=1, Maximum=1)
     )
     B=&MakeVector(
      X=1
      Y=1
      Z=&FloatFromCurve(
       FloatCurve=DataInterface() {propertyValues {Curve {keys [
           {interpMode=RCIM_Cubic, tangentMode=RCTM_User, time=0, value=0}
           {interpMode=RCIM_Cubic, tangentMode=RCTM_User, time=0.5, value=1}
           {tangentMode=RCTM_User, time=1, value=0}
          ]}}}
       Scale Curve=&Lerp_Float(
        A=1
        B=&FN_FrontEndFloatModifier(Gameplay=2, Front End=2)
        Alpha=&ClampFloat(Float=$System.Audio NPC)
       )
      )
     )
    )) {isSetParameters=true, module=SetVariables_#5}
   /SetVariables(Particles.Mesh size02:Vector3f=&Multiply_Vector(
     A=&Multiply_VectorByFloat(
      Vector=&RandomRangeVector(Minimum=<0.1 0.1 1>, Maximum=<0.1 0.1 1>)
      Float=&RandomRangeFloat(Minimum=1, Maximum=1)
     )
     B=&MakeVector(
      X=1
      Y=1
      Z=&FloatFromCurve(
       FloatCurve=DataInterface() {propertyValues {Curve {keys [
           {interpMode=RCIM_Cubic, tangentMode=RCTM_User, time=0, value=0}
           {interpMode=RCIM_Cubic, tangentMode=RCTM_User, time=0.5, value=1}
           {tangentMode=RCTM_User, time=1, value=0}
          ]}}}
       Scale Curve=&Lerp_Float(
        A=2
        B=&FN_FrontEndFloatModifier(Gameplay=2.5, Front End=2.2)
        Alpha=&ClampFloat(Float=$NPC.FortniteNPC.EventAudio_Amplitude)
       )
      )
     )
    )) {isSetParameters=true}
   /SetVariables(Particles.Mesh02 Color:LinearColor=&MakeLinearColorFromVectorAndFloat(
     "Vector (RGB)"=&Multiply_VectorByFloat(
      Vector=&MakeVectorFromLinearColorRGB(Color=&ColorFromCurve(
        ColorCurve=DataInterface() {propertyValues {
          AlphaCurve {keys [<at 0 1 RCIM_Cubic>]}
          BlueCurve {keys [
            <at 0.0075 0>
            <at 0.2424 0.0417>
            <at 0.2989 1>
            <at 0.4733 1>
            <at 0.6883 1>
            {arriveTangent=-3.3972, interpMode=RCIM_Cubic, leaveTangent=-3.3972, time=0.7448, value=0.0417}
            <at 0.9827 0>
           ]}
          GreenCurve {keys [
            <at 0.0075 0.6667>
            <at 0.2424 0.5208>
            <at 0.2989 0.0833>
            <at 0.4733 0.1667>
            <at 0.6883 0>
            {arriveTangent=2.2648, interpMode=RCIM_Cubic, leaveTangent=2.2648, time=0.7448, value=0.5208}
            <at 0.9827 0.6667>
           ]}
          RedCurve {keys [
            <at 0.0075 1>
            <at 0.2424 1>
            <at 0.2989 1>
            <at 0.4733 0>
            <at 0.6883 0.1667>
            {arriveTangent=2.831, interpMode=RCIM_Cubic, leaveTangent=2.831, time=0.7448, value=1}
            <at 0.9827 1>
           ]}
         }}
        CurveIndex=&Multiply_Float(A=$Particles.Random, B=$Particles.NormalizedAge)
       ))
      Float=&Multiply_Float(A=&FN_FrontEndFloatModifier(Gameplay=2, Front End=1), B=$System.TOD)
     )
     "Float (Alpha)"=&FloatFromCurve(
      FloatCurve=DataInterface() {propertyValues {Curve {keys [
          {interpMode=RCIM_Cubic, tangentMode=RCTM_User, time=0, value=0}
          {interpMode=RCIM_Cubic, tangentMode=RCTM_User, time=0.5, value=1}
          {tangentMode=RCTM_User, time=1, value=0}
         ]}}}
      Scale Curve=&Multiply_Float(A=&FN_FrontEndFloatModifier(Gameplay=1, Front End=2), B=4)
     )
    )) {isSetParameters=true}
   /SetVariables(Particles.VisibilityTag=&MakeCustomIntFromBool(Bool=$User.bisHandHidden, True INT=1, False INT=0)) {isSetParameters=true, module=SetVariables_#6}
  ]
  rendererClasses [&NiagaraMeshRendererProperties]
  renderers [
   {rendererClass=&NiagaraMeshRendererProperties, rendererIndex=0}
   {rendererClass=&NiagaraMeshRendererProperties, rendererIndex=1}
   {rendererClass=&NiagaraMeshRendererProperties, rendererIndex=2}
  ]
 }
 Blocks_left_Ray002 {
  emitterUpdateScript [
   &EmitterState(
    Inactive Response=::NewEnumerator1
    Life Cycle Mode=::NewEnumerator1
    Loop Behavior=::NewEnumerator1
    Loop Duration Mode=::NewEnumerator1
    UseLoopDelay=false
   )
   &SpawnRate(
    SpawnRate=&MakeCustomFloatFromBool(Bool=&SetBoolByFloatComparison(A=$System.Audio NPC, B=0), True Float=50, False Float=0)
    Use Spawn Probability=false
   )
  ]
  particleSpawnScript [
   &FN_SetSpawnGroup(Number of Spawn Groups=3)
   &InitializeParticle(
    Color=<rgba 1.2 1.2 1.2 2>
    Lifetime=0.2
    Mesh Scale=&Multiply_VectorByFloat(
     Vector=&RandomRangeVector(Minimum=<0.2 0.2 0.4>, Maximum=<0.3 0.3 0.8>)
     Float=&RandomRangeFloat(Minimum=1, Maximum=1)
    )
    Mesh Scale Mode=::NewEnumerator4
    Sprite Size Mode=::NewEnumerator3
    UsePositionOffset=false
    Write Material Random=true
   )
   &SphereLocation(
    MaskSpawnGroup=false
    Override Seed=false
    Sphere Origin:NiagaraPosition=&ConvertVectorToPosition(Input Position=<0 0 0>, Vector Is In Absolute Large World Space:NiagaraBool=false)
    Sphere Radius=1
    Surface Only=false
    UseNonUniformScale=false
    UseOffset=false
    UseSphereOrientation=false
   )
   /SetVariables(Particles.Random=&RandomRangeFloat(Minimum=0, Maximum=1)) {isSetParameters=true, module=SetVariables_#3}
   &InitialMeshOrientation(Mesh Orientation Mode=::NewEnumerator2, Rotation=<0 0 0.25>, Use Rotation Vector=true)
  ]
  particleUpdateScript [
   &ParticleState()
   &SolveForcesAndVelocity(Clamp Velocity=false, Limit Acceleration=false)
   &SkeletalMeshLocation(
    Bone Sampling Mode=::NewEnumerator4
    MaskSpawnGroup=false
    Orientation Sampling=::NewEnumerator1
    Sampled Position Offset=&Add_Vector(
     A=&ConvertPositionToVector(Input Position=$Particles.Initial.Position, Convert To Absolute Large World Vector=false)
     B=&LerpPosition(A=<-23 0 -2>, B=<0 0 -2>, Alpha=$Particles.Random)
    )
    Skeletal Mesh=DataInterface() {propertyValues {
      DefaultMesh=null
      ExcludeBoneName=root
      OptionalComponentClass=null
      PreviewMesh=&F_MED_Cactus_Rocker
      SkinningMode=SkinOnTheFly
      SoftSourceActor=null
      SourceMode=Default
      UvSetIndex=0
      WholeMeshLOD=-1
      bExcludeBone=false
      bReadDeformedGeometry=true
      bRequireCurrentFrameData=true
      ComponentTags []
      FilteredBones [lowerarm_r]
      FilteredSockets []
      MeshUserParameter {parameter {name=None, typeDefHandle {registeredTypeIndex=1}, varData []}}
      SamplingRegions []
     }}
   )
   &ScaleColor(
    Scale Alpha=&Multiply_Float(
     A=&FloatFromCurve(
      FloatCurve=DataInterface() {propertyValues {Curve {keys [
          {interpMode=RCIM_Cubic, tangentMode=RCTM_User, time=0, value=0}
          {interpMode=RCIM_Cubic, tangentMode=RCTM_User, time=0.5, value=1}
          {tangentMode=RCTM_User, time=1, value=0}
         ]}}}
      Scale Curve=&Multiply_Float(
       A=&FN_FrontEndFloatModifier(Gameplay=&Multiply_Float(A=1, B=1), Front End=1)
       B=2
      )
     )
     B=&FN_FrontEndFloatModifier(Gameplay=1.25, Front End=0.75)
    )
    Scale RGB=&Multiply_VectorByFloat(
     Vector=&MakeVectorFromLinearColorRGB(Color=&ColorFromCurve(
       ColorCurve=DataInterface() {propertyValues {
         AlphaCurve {keys [<at 0 1 RCIM_Cubic>]}
         BlueCurve {keys [
           <at 0.0075 0>
           <at 0.2424 0.0417>
           <at 0.2989 1>
           <at 0.6883 1>
           {arriveTangent=-3.3972, interpMode=RCIM_Cubic, leaveTangent=-3.3972, time=0.7448, value=0.0417}
           <at 0.9827 0>
          ]}
         GreenCurve {keys [
           <at 0.0075 0.6667>
           <at 0.2424 0.8403>
           <at 0.2989 0.0833>
           <at 0.6883 0>
           {arriveTangent=2.2648, interpMode=RCIM_Cubic, leaveTangent=2.2648, time=0.7448, value=0.8403}
           <at 0.9827 0.6667>
          ]}
         RedCurve {keys [
           <at 0.0075 1>
           <at 0.2424 1>
           <at 0.2989 0.5417>
           <at 0.6883 0>
           {arriveTangent=3.3972, interpMode=RCIM_Cubic, leaveTangent=3.3972, time=0.7448, value=1}
           <at 0.9827 1>
          ]}
        }}
       CurveIndex=&Multiply_Float(A=$Particles.Random, B=$Particles.NormalizedAge)
      ))
     Float=&Multiply_Float(
      A=&ClampFloat(Float=$System.Audio NPC, Min=1)
      B=&FN_FrontEndFloatModifier(Gameplay=1, Front End=0.75)
     )
    )
    ScaleA=true
    ScaleRGB=true
   )
   &ScaleMeshSize(Scale Factor=&MakeVector(
     X=1
     Y=1
     Z=&FloatFromCurve(
      FloatCurve=DataInterface() {propertyValues {Curve {keys [
          {interpMode=RCIM_Cubic, tangentMode=RCTM_User, time=0, value=0}
          {interpMode=RCIM_Cubic, tangentMode=RCTM_User, time=0.5, value=1}
          {tangentMode=RCTM_User, time=1, value=0}
         ]}}}
      Scale Curve=&Lerp_Float(
       A=1
       B=&FN_FrontEndFloatModifier(Gameplay=1.5, Front End=1.5)
       Alpha=&ClampFloat(Float=$System.Audio NPC)
      )
     )
    ))
   /SetVariables(Particles.Glow Color=&MakeLinearColorFromVectorAndFloat(
     "Vector (RGB)"=&Multiply_VectorByFloat(
      Vector=&Multiply_VectorByFloat(
       Vector=&MakeVectorFromLinearColorRGB(Color=&ColorFromCurve(
         ColorCurve=DataInterface() {propertyValues {
           AlphaCurve {keys [<at 0 1 RCIM_Cubic>]}
           BlueCurve {keys [
             <at 0.0075 0>
             <at 0.0927 0>
             <at 0.2424 0.0417>
             <at 0.2989 1>
             <at 0.4733 1>
             <at 0.6883 1>
             {arriveTangent=-6.0278, interpMode=RCIM_Cubic, leaveTangent=-6.0278, time=0.7448, value=0.0417}
             <at 0.8542 0>
             <at 0.9827 0>
            ]}
           GreenCurve {keys [
             <at 0.0075 0.6667>
             <at 0.0927 0.1667>
             <at 0.2424 0.5208>
             <at 0.2989 0.0833>
             <at 0.4733 0.1667>
             <at 0.6883 0>
             {arriveTangent=1.0046, interpMode=RCIM_Cubic, leaveTangent=1.0046, time=0.7448, value=0.5208}
             <at 0.8542 0.1667>
             <at 0.9827 0.6667>
            ]}
           RedCurve {keys [
             <at 0.0075 1>
             <at 0.0927 1>
             <at 0.2424 1>
             <at 0.2989 1>
             <at 0.4733 0>
             <at 0.6883 0.1667>
             {arriveTangent=5.0231, interpMode=RCIM_Cubic, leaveTangent=5.0231, time=0.7448, value=1}
             <at 0.8542 1>
             <at 0.9827 1>
            ]}
          }}
         CurveIndex=&Multiply_Float(A=$Particles.Random, B=$Particles.NormalizedAge)
        ))
       Float=&Multiply_Float(
        A=&ClampFloat(Float=$System.Audio NPC, Min=1)
        B=&FN_FrontEndFloatModifier(Gameplay=1, Front End=0.85)
       )
      )
      Float=0.6
     )
     "Float (Alpha)"=&Multiply_Float(
      A=&Multiply_Float(
       A=&FloatFromCurve(
        FloatCurve=DataInterface() {propertyValues {Curve {keys [
            {interpMode=RCIM_Cubic, tangentMode=RCTM_User, time=0, value=0}
            <at 0.2 1>
            {interpMode=RCIM_Cubic, tangentMode=RCTM_User, time=0.8, value=1}
            {tangentMode=RCTM_User, time=1, value=0}
           ]}}}
        Scale Curve=&Multiply_Float(
         A=&FN_FrontEndFloatModifier(Gameplay=&Multiply_Float(A=1, B=1), Front End=1)
         B=&RandomRangeFloat(Minimum=1, Maximum=1)
        )
       )
       B=2
      )
      B=&FN_FrontEndFloatModifier(Gameplay=2, Front End=3)
     )
    )) {isSetParameters=true, module=SetVariables_#4}
   /SetVariables(Particles.Glow Size:Vector3f=&Multiply_Vector(
     A=&Multiply_VectorByFloat(
      Vector=&RandomRangeVector(Minimum=<0.3 0.3 0.7>, Maximum=<0.5 0.5 1.2>)
      Float=&RandomRangeFloat(Minimum=1, Maximum=1)
     )
     B=&MakeVector(
      X=1
      Y=1
      Z=&FloatFromCurve(
       FloatCurve=DataInterface() {propertyValues {Curve {keys [
           {interpMode=RCIM_Cubic, tangentMode=RCTM_User, time=0, value=0}
           {interpMode=RCIM_Cubic, tangentMode=RCTM_User, time=0.5, value=1}
           {tangentMode=RCTM_User, time=1, value=0}
          ]}}}
       Scale Curve=&Lerp_Float(
        A=1
        B=&FN_FrontEndFloatModifier(Gameplay=2, Front End=2)
        Alpha=&ClampFloat(Float=$System.Audio NPC)
       )
      )
     )
    )) {isSetParameters=true, module=SetVariables_#5}
   /SetVariables(Particles.Mesh size02:Vector3f=&Multiply_Vector(
     A=&Multiply_VectorByFloat(
      Vector=&RandomRangeVector(Minimum=<0.1 0.1 1>, Maximum=<0.1 0.1 1>)
      Float=&RandomRangeFloat(Minimum=1, Maximum=1)
     )
     B=&MakeVector(
      X=1
      Y=1
      Z=&FloatFromCurve(
       FloatCurve=DataInterface() {propertyValues {Curve {keys [
           {interpMode=RCIM_Cubic, tangentMode=RCTM_User, time=0, value=0}
           {interpMode=RCIM_Cubic, tangentMode=RCTM_User, time=0.5, value=1}
           {tangentMode=RCTM_User, time=1, value=0}
          ]}}}
       Scale Curve=&Lerp_Float(
        A=2
        B=&FN_FrontEndFloatModifier(Gameplay=2.5, Front End=2.2)
        Alpha=&ClampFloat(Float=$NPC.FortniteNPC.EventAudio_Amplitude)
       )
      )
     )
    )) {isSetParameters=true}
   /SetVariables(Particles.Mesh02 Color:LinearColor=&MakeLinearColorFromVectorAndFloat(
     "Vector (RGB)"=&Multiply_VectorByFloat(
      Vector=&MakeVectorFromLinearColorRGB(Color=&ColorFromCurve(
        ColorCurve=DataInterface() {propertyValues {
          AlphaCurve {keys [<at 0 1 RCIM_Cubic>]}
          BlueCurve {keys [
            <at 0.0075 0>
            <at 0.2424 0.0417>
            <at 0.2989 1>
            <at 0.4733 1>
            <at 0.6883 1>
            {arriveTangent=-3.3972, interpMode=RCIM_Cubic, leaveTangent=-3.3972, time=0.7448, value=0.0417}
            <at 0.9827 0>
           ]}
          GreenCurve {keys [
            <at 0.0075 0.6667>
            <at 0.2424 0.5208>
            <at 0.2989 0.0833>
            <at 0.4733 0.1667>
            <at 0.6883 0>
            {arriveTangent=2.2648, interpMode=RCIM_Cubic, leaveTangent=2.2648, time=0.7448, value=0.5208}
            <at 0.9827 0.6667>
           ]}
          RedCurve {keys [
            <at 0.0075 1>
            <at 0.2424 1>
            <at 0.2989 1>
            <at 0.4733 0>
            <at 0.6883 0.1667>
            {arriveTangent=2.831, interpMode=RCIM_Cubic, leaveTangent=2.831, time=0.7448, value=1}
            <at 0.9827 1>
           ]}
         }}
        CurveIndex=&Multiply_Float(A=$Particles.Random, B=$Particles.NormalizedAge)
       ))
      Float=&Multiply_Float(A=&FN_FrontEndFloatModifier(Gameplay=2, Front End=1), B=$System.TOD)
     )
     "Float (Alpha)"=&FloatFromCurve(
      FloatCurve=DataInterface() {propertyValues {Curve {keys [
          {interpMode=RCIM_Cubic, tangentMode=RCTM_User, time=0, value=0}
          {interpMode=RCIM_Cubic, tangentMode=RCTM_User, time=0.5, value=1}
          {tangentMode=RCTM_User, time=1, value=0}
         ]}}}
      Scale Curve=&Multiply_Float(A=&FN_FrontEndFloatModifier(Gameplay=1, Front End=2), B=4)
     )
    )) {isSetParameters=true}
   /SetVariables(Particles.VisibilityTag=&MakeCustomIntFromBool(Bool=$User.bisHandHidden, True INT=1, False INT=0)) {isSetParameters=true, module=SetVariables_#6}
  ]
  rendererClasses [&NiagaraMeshRendererProperties]
  renderers [
   {rendererClass=&NiagaraMeshRendererProperties, rendererIndex=0}
   {rendererClass=&NiagaraMeshRendererProperties, rendererIndex=1}
   {rendererClass=&NiagaraMeshRendererProperties, rendererIndex=2}
  ]
 }
 Light001 {
  emitterUpdateScript [
   &EmitterState(Inactive Response=::NewEnumerator1, Life Cycle Mode=::NewEnumerator1, Loop Duration=1, UseLoopDelay=false)
   &SpawnBurst_Instantaneous(
    Spawn Count=&MakeIntFromFloat(Float=&MakeCustomFloatFromBool(Bool=&SetBoolByFloatComparison(A=$NPC.FortniteNPC.EventAudio_Amplitude, B=0), True Float=2, False Float=0))
    Use Spawn Probability=false
    UseLoopCountLimit=false
   )
  ]
  particleSpawnScript [
   &FN_SetSpawnGroup(Number of Spawn Groups=2)
   &InitializeParticle(Sprite Size Mode=::NewEnumerator3, Uniform Sprite Size=50, UsePositionOffset=false, Write Material Random=true)
  ]
  particleUpdateScript [
   &ParticleState()
   &SkeletalMeshLocation(
    Bone / Socket Index=$Particles.SpawnGroup
    Bone Sampling Mode=::NewEnumerator4
    MaskSpawnGroup=false
    Sampled Position Offset=&SelectVectorFromArray(Direct Array Index=$Particles.SpawnGroup, Vector Selection Array=DataInterface() {propertyValues {FloatData [<20 0 20>, <-20 0 -20>]}})
    Skeletal Mesh=DataInterface() {propertyValues {
      DefaultMesh=null
      ExcludeBoneName=root
      OptionalComponentClass=null
      PreviewMesh=&F_MED_Cactus_Rocker
      SkinningMode=SkinOnTheFly
      SoftSourceActor=null
      SourceMode=Default
      UvSetIndex=0
      WholeMeshLOD=-1
      bExcludeBone=false
      bReadDeformedGeometry=true
      bRequireCurrentFrameData=true
      ComponentTags []
      FilteredBones [upperarm_l,upperarm_r]
      FilteredSockets []
      MeshUserParameter {parameter {name=None, typeDefHandle {registeredTypeIndex=1}, varData []}}
      SamplingRegions []
     }}
   )
   &FN_Light_Attributes(
    Color=&ColorFromCurve(ColorCurve=DataInterface() {propertyValues {
       AlphaCurve {keys [<at 0 1 RCIM_Cubic>]}
       BlueCurve {keys [
         <at 0 1>
         <at 0.2335 0>
         <at 0.4197 1>
         {arriveTangent=-2.4048, interpMode=RCIM_Cubic, leaveTangent=-2.4048, time=0.7673, value=1}
         <at 0.8355 0>
        ]}
       GreenCurve {keys [
         <at 0 0>
         <at 0.2335 0.5>
         <at 0.4197 0>
         {arriveTangent=1.2024, interpMode=RCIM_Cubic, leaveTangent=1.2024, time=0.7673, value=0}
         <at 0.8355 0.5>
        ]}
       RedCurve {keys [
         <at 0 1>
         <at 0.2335 1>
         <at 0.4197 0.1667>
         {arriveTangent=2.004, interpMode=RCIM_Cubic, leaveTangent=2.004, time=0.7673, value=1}
         <at 0.8355 1>
        ]}
      }})
    Create Color Attribute=true
    Create Exponent Attribute=true
    Create Position Attribute=true
    Create Radius Attribute=true
    Exponent=3
    Position=&AddVectorToPosition(Position=$Particles.Position, Vector=<0 -3 0>)
    Radius=100
    Scale Color=&Multiply_Float(
     A=&Multiply_Float(A=4, B=&FNTimeOfDayScalar(Night Scale=0.7, Morning Scale=1, Evening Scale=1))
     B=&Lerp_Float(A=0, B=1, Alpha=$NPC.FortniteNPC.EventAudio_Amplitude)
    )
   )
  ]
  rendererClasses [&NiagaraSpriteRendererProperties, &NiagaraLightRendererProperties]
  renderers [
   {rendererClass=&NiagaraSpriteRendererProperties}
   {rendererClass=&NiagaraLightRendererProperties, rendererIndex=1}
  ]
 }
 ParticleDecal {
  emitterUpdateScript [
   &EmitterState(
    Inactive Response=::NewEnumerator1
    Life Cycle Mode=::NewEnumerator1
    Loop Behavior=::NewEnumerator1
    Loop Duration Mode=::NewEnumerator1
    UseLoopDelay=false
   )
   &SpawnBurst_Instantaneous(Spawn Count=6, Use Spawn Probability=false, UseLoopCountLimit=false)
  ]
  particleSpawnScript [
   &FN_SetSpawnGroup(Number of Spawn Groups=6)
   &InitializeParticle(Color=<rgba 7 7 7 1>, Sprite Size Mode=::NewEnumerator3, Uniform Sprite Size=25, UsePositionOffset=false, Write Material Random=true)
   &CameraOffset(Camera Offset Amount=20)
  ]
  particleUpdateScript [
   &ParticleState(
    Kill Particles When Lifetime Has Elapsed=false
    Let Infinitely Lived Particles Die When Emitter Deactivates=true
    Loop Particles Lifetime=true
   )
   &SkeletalMeshLocation(
    Bone / Socket Index=$Particles.SpawnGroup
    Bone Sampling Mode=::NewEnumerator4
    MaskSpawnGroup=false
    Sampled Position Offset=&SelectVectorFromArray(
     Direct Array Index=$Particles.SpawnGroup
     Vector Selection Array=DataInterface() {propertyValues {FloatData [<15 0 0>, <-15 0 0>, <0 0 0>, <0 0 0>, <-10 0 0>, <10 0 0>]}}
    )
    Skeletal Mesh=DataInterface() {propertyValues {
      DefaultMesh=null
      ExcludeBoneName=root
      OptionalComponentClass=null
      PreviewMesh=&F_MED_Cactus_Rocker
      SkinningMode=SkinOnTheFly
      SoftSourceActor=null
      SourceMode=Default
      UvSetIndex=0
      WholeMeshLOD=-1
      bExcludeBone=false
      bReadDeformedGeometry=true
      bRequireCurrentFrameData=true
      ComponentTags []
      FilteredBones [upperarm_l,upperarm_r,upperarm_l,upperarm_r,hand_l,hand_r]
      FilteredSockets []
      MeshUserParameter {parameter {name=None, typeDefHandle {registeredTypeIndex=1}, varData []}}
      SamplingRegions []
     }}
   ) {module=SkeletalMeshLocation001}
   &ScaleColor(
    Scale RGB=&MakeVectorFromLinearColorRGB(Color=&MultiplyLinearColor(
      Linear Color=&ColorFromCurve(ColorCurve=DataInterface() {propertyValues {
         AlphaCurve {keys [<at 0 1 RCIM_Cubic>]}
         BlueCurve {keys [<at 0 1>,<at 0.2484 1>,<at 0.3668 0>,<at 0.53 1>,<at 0.6878 0>,<at 0.9085 1 RCIM_Cubic>]}
         GreenCurve {keys [<at 0 0>,<at 0.2484 0>,<at 0.3668 0.5>,<at 0.53 0>,<at 0.6878 0.3333>,<at 0.9085 0 RCIM_Cubic>]}
         RedCurve {keys [<at 0 1>,<at 0.2484 0.1667>,<at 0.3668 1>,<at 0.53 0.1667>,<at 0.6878 1>,<at 0.9085 1 RCIM_Cubic>]}
        }})
      Scale Factor=&MakeFloat4FromVectorAndFloat(
       "Vector (RGB)"=&VectorFromFloat(Value=&Multiply_Float(
         A=&Multiply_Float(A=1, B=&FNTimeOfDayScalar(Night Scale=0.7, Morning Scale=1, Evening Scale=1))
         B=&Lerp_Float(A=0, B=1, Alpha=$NPC.FortniteNPC.EventAudio_Amplitude)
        ))
       "Float (Alpha)"=1
      )
     ))
    ScaleA=true
    ScaleRGB=true
   )
   /SetVariables(Particles.VisibilityTag=&MakeCustomIntFromBool(
     Bool=&SetBoolByIntComparison(A=$Particles.SpawnGroup, B=4, Comparison Type=::NewEnumerator2)
     True INT=&MakeCustomIntFromBool(Bool=$User.bisHandHidden, True INT=1, False INT=0)
     False INT=0
    )) {isSetParameters=true}
  ]
  rendererClasses [&NiagaraSpriteRendererProperties]
  renderers [{rendererClass=&NiagaraSpriteRendererProperties}]
 }
}
summary {
 systemName=NS_MED_TigerRootFame_Orange
 emitters [
  {emitterName=Blocks_left_Ray, rendererClasses [&NiagaraMeshRendererProperties]}
  {emitterName=Blocks_Right_Ray, rendererClasses [&NiagaraMeshRendererProperties]}
  {emitterName=Blocks_left_Ray001, rendererClasses [&NiagaraMeshRendererProperties]}
  {emitterName=Blocks_left_Ray002, rendererClasses [&NiagaraMeshRendererProperties]}
  {emitterName=ParticleDecal, rendererClasses [&NiagaraSpriteRendererProperties]}
  {emitterName=Light001, rendererClasses [&NiagaraSpriteRendererProperties, &NiagaraLightRendererProperties]}
 ]
 userVariables [
  User.AuidoPreview {type=NiagaraFloat, defaultValue=0, description=""}
  User.bIsFrontEnd {type=NiagaraBool}
  User.bIsFrontEndPreview {type=NiagaraBool}
  User.bisHandHidden {type=NiagaraBool}
 ]
}
