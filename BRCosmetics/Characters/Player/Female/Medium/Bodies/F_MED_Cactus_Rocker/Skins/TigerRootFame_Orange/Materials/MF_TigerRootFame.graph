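# Material function for the TigerRootFame character skin that layers audio-reactive, colored glow effects onto incoming material attributes.
# FX-texture channels and pre-skinned bounds split the body into regions, tinted by contrast-adjusted Fresnel and curve-atlas colors.
# Four procedural sine waves add banded color, with emote/event audio parameters and MF_AudioReactivity supplying audio signals.
# Static switches select hair, eye and simple-masking variants, setting base color, specular and time-of-day-corrected emissive.
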MaterialFunction /BRCosmetics/Characters/Player/Female/Medium/Bodies/F_MED_Cactus_Rocker/Skins/TigerRootFame_Orange/Materials/MF_TigerRootFame
graph {nodes [
  n1 = Add(A=NamedRerouteUsage(), B=NamedRerouteUsage())
  Add(A=NamedRerouteUsage(), B=@n1)
  ComponentMask() {g=true, r=true}
  Constant()
  n2 = NamedRerouteUsage()
  n3 = Add(A=NamedRerouteUsage(), B=NamedRerouteUsage())
  n4 = NamedRerouteUsage()
  n5 = OneMinus(None=Saturate(None=Add(A=@n3, B=@n4)))
  n6 = StaticSwitchParameter(True=Multiply(A=OneMinus(None=@n2), B=NamedRerouteUsage()), False=Multiply(A=NamedRerouteUsage(), B=@n5)) {group=Tiger_Root, parameterName=TRoot_Hair}
  n7 = NamedRerouteUsage()
  n8 = Add(A=NamedRerouteUsage(), B=NamedRerouteUsage())
  n9 = Multiply(
   A=StaticSwitchParameter(True=NamedRerouteUsage(), False=Add(A=@n7, B=@n8)) {group=Tiger_Root, parameterName=TRoot_Hair}
   B=NamedRerouteUsage()
  )
  n10 = LinearInterpolate(B=@n9, Alpha=NamedRerouteUsage())
  n11 = NamedRerouteUsage()
  n12 = Multiply(A=ComponentMask(None=@n11) {r=true}, B=ComponentMask(None=@n11) {g=true})
  n13 = VectorParameter() {group=Tiger_Root, parameterName=TRoot_Eye_ColorB, defaultValue=<rgba 1 0 0.0937 1>}
  n14 = Multiply(A=@n12, B=@n13.A)
  n15 = OneMinus(None=@n14)
  n16 = Constant() {r=1}
  n17 = Multiply(A=@n5, B=NamedRerouteUsage())
  n18 = NamedRerouteUsage()
  n19 = Add(
   A=StaticSwitchParameter(True=Constant(), False=LinearInterpolate(B=@n17, Alpha=@n18)) {group=Tiger_Root, parameterName=TRoot_Hair}
   B=LinearInterpolate(A=@n9, Alpha=NamedRerouteUsage()) {constB=0}
  )
  n20 = LinearInterpolate(
   A=@n19
   B=LinearInterpolate(
    A=VectorParameter() {group=Tiger_Root, parameterName=TRoot_Eye_ColorA, defaultValue=<rgba 1 0 0.0937 1>}.RGB
    B=@n13.RGB
    Alpha=NamedRerouteUsage()
   )
   Alpha=@n14
  )
  n21 = NamedRerouteUsage()
  n22 = OneMinus(None=@n12)
  n23 = SetMaterialAttributes!(
   MaterialAttributes=NamedRerouteUsage()
   BaseColor=Multiply(A=Add(A=@n6, B=@n10), B=StaticSwitchParameter(True=@n15, False=@n16) {group=Tiger_Root, parameterName=TRoot_Eyes})
   EmissiveColor=MaterialFunctionCall!(Color=StaticSwitchParameter(True=@n20, False=@n19) {group=Tiger_Root, parameterName=TRoot_Eyes}) {materialFunction=/Game/Packages/Fortress_SharedMaterials/Anime_Master_Material/MF_TODColorCurve}.Result
   Normal
   Specular=StaticSwitchParameter(True=Multiply(A=@n21, B=@n22), False=@n21) {group=Tiger_Root, parameterName=TRoot_Eyes}
  ) {attributeSetTypes [
    #1
    #2
    #3
    #4
   ]}
  n24 = SetMaterialAttributes(
   MaterialAttributes=NamedRerouteUsage()
   BaseColor=NamedRerouteUsage()
   EmissiveColor=MaterialFunctionCall(Color=NamedRerouteUsage()) {materialFunction=&MF_TODColorCurve}.Result
  ) {attributeSetTypes [
    #1
    #2
    #3
    #4
   ]}
  FunctionOutput(None=StaticSwitchParameter(True=@n23, False=@n24) {group=Tiger_Root, parameterName=Troot__Simpe_Msking})
  n25 = MaterialFunctionCall() {materialFunction=/Game/Packages/Fortress_SharedMaterials/Base_Material_Functions/BaseCharacterShaderFunctions/Audio/MF_AudioReactivity}
  Multiply(A=@n25.Amplitude, B=@n25.Amplitude)
  n26 = TextureSample() {texture=/BRCosmetics/Characters/Player/Female/Medium/Bodies/F_MED_Cactus_Rocker/Skins/TigerRootFame_Orange/Textures/T_F_MED_TigerRootFame_Orange_Body_D}
  n27 = TextureSample() {
   samplerType=SAMPLERTYPE_Normal
   texture=/BRCosmetics/Characters/Player/Female/Medium/Bodies/F_MED_Cactus_Rocker/Skins/TigerRootFame_Orange/Textures/T_F_MED_TigerRootFame_Orange_Body_N
  }
  n28 = TextureSample() {
   samplerType=SAMPLERTYPE_Masks
   texture=/BRCosmetics/Characters/Player/Female/Medium/Bodies/F_MED_Cactus_Rocker/Skins/TigerRootFame_Orange/Textures/T_F_MED_TigerRootFame_Orange_Body_S
  }
  n29 = SetMaterialAttributes!(MaterialAttributes, BaseColor=@n26.RGB, EmissiveColor, Normal=@n27.RGB, Metallic=@n28.G, Specular=@n28.G, Roughness=@n28.G) {attributeSetTypes [
    #1
    #2
    #3
    #5
    #4
    #6
   ]}
  n30 = GetMaterialAttributes(None=FunctionInput(Preview=@n29) {bUsePreviewValueAsDefault=true, inputName=Mat, inputType=FunctionInput_MaterialAttributes}) {attributeGetTypes [
    #1
    #2
    #3
    #4
   ]}
  NamedRerouteDeclaration(None=@n30.MaterialAttributes) {name=Root_Material, nodeColor=<rgba 0 0.1797 1 1>}
  NamedRerouteDeclaration(None=@n30.BaseColor) {name=Root_BaseColor, nodeColor=<rgba 0 0.1797 1 1>}
  n31 = NamedRerouteUsage()
  n32 = Ceil(None=Add(A=ComponentMask(None=@n31) {b=true}, B=ScalarParameter() {defaultValue=-0.15, group=Tiger_Root, parameterName=TRoot_Masking_Red_Lower}))
  NamedRerouteDeclaration(None=Saturate(None=OneMinus(None=@n32))) {name=Msk_Lower, nodeColor=<rgba 0 0.3438 1 1>}
  n33 = NamedRerouteUsage()
  n34 = Add(A=Add(A=NamedRerouteUsage(), B=NamedRerouteUsage()), B=Multiply(A=Add(A=@n33, B=@n1), B=NamedRerouteUsage()))
  n35 = ScalarParameter() {group=Tiger_Root, parameterName=TRoot--Emissive--BaseColor}
  n36 = NamedRerouteUsage()
  n37 = NamedRerouteUsage()
  NamedRerouteDeclaration(None=Add(A=LinearInterpolate(A=@n34, Alpha=@n35) {constB=0}, B=LinearInterpolate(B=@n36, Alpha=@n37))) {name=Emissive, nodeColor=<rgba 0 0.6719 1 1>}
  n38 = NamedRerouteUsage()
  n39 = Saturate(None=Multiply(A=ComponentMask(None=@n38) {g=true}, B=NamedRerouteUsage()))
  n40 = Power(Base=NamedRerouteUsage(), Exp=ScalarParameter() {defaultValue=1, group=Tiger_Root, parameterName=TRoot_GreenUpper_Fresnel_Power})
  n41 = ScalarParameter() {defaultValue=1, group=Tiger_Root, parameterName=TRoot_GreenUpper_Fresnel_Contrast}
  n42 = Multiply(
   A=@n39
   B=Saturate(None=MaterialFunctionCall(In=@n40, Contrast=@n41) {materialFunction=&CheapContrast}.Result)
  )
  NamedRerouteDeclaration(None=@n42) {name=MSk_GreenUpper, nodeColor=<rgba 1 0 0.1875 1>}
  n43 = NamedRerouteUsage()
  n44 = ComponentMask(None=@n43) {g=true}
  n45 = ComponentMask(None=@n43) {r=true}
  n46 = Power(Base=NamedRerouteUsage(), Exp=ScalarParameter() {defaultValue=1, group=Tiger_Root, parameterName=TRoot_Green_Fresnel_Power})
  n47 = ScalarParameter() {defaultValue=1, group=Tiger_Root, parameterName=TRoot_Green_Fresnel_Contrast}
  n48 = Multiply(
   A=Saturate(None=Subtract(A=@n44, B=@n45))
   B=Saturate(None=MaterialFunctionCall(In=@n46, Contrast=@n47) {materialFunction=&CheapContrast}.Result)
  )
  NamedRerouteDeclaration(None=@n48) {name=MSk_Green, nodeColor=<rgba 1 0 0.1875 1>}
  n49 = Power(Base=NamedRerouteUsage(), Exp=ScalarParameter() {defaultValue=1.5, group=Tiger_Root, parameterName=TRoot_Red_Fresnel_Power})
  n50 = ScalarParameter() {defaultValue=1, group=Tiger_Root, parameterName=TRoot_Red_Fresnel_Contrast}
  n51 = Multiply(
   A=ComponentMask(None=NamedRerouteUsage()) {r=true}
   B=Saturate(None=MaterialFunctionCall(In=@n49, Contrast=@n50) {materialFunction=&CheapContrast}.Result)
  )
  NamedRerouteDeclaration(None=@n51) {name=MSk_Red, nodeColor=<rgba 1 0 0.1875 1>}
  n52 = CollectionParameter() {collection=&FortniteMaterialParameters, parameterName=EventAudio_Amplitude}
  n53 = CollectionParameter() {collection=&FortniteMaterialParameters, parameterName=Emote_AudioLow}
  n54 = Max(A=@n52, B=@n53)
  n55 = ScalarParameter() {defaultValue=0.25, group=Tiger_Root_Audio, parameterName=TRoot_SPectrum_Low_Remap, sortPriority=150}
  n56 = ScalarParameter() {defaultValue=0.8, group=Tiger_Root_Audio, parameterName=TRoot_Spectrum_High_Remap, sortPriority=150}
  n57 = Constant()
  n58 = Constant() {r=1}
  n59 = CollectionParameter() {collection=&FortniteMaterialParameters, parameterName=Emote_AudioHigh}
  n60 = Max(
   A=Saturate(None=MaterialFunctionCall(Input=@n54, Input Low=@n55, Input High=@n56, Target Low=@n57, Target High=@n58) {materialFunction=&RemapValueRange}.Result)
   B=@n59
  )
  NamedRerouteDeclaration(None=@n60) {name=Audio_Data, nodeColor=<rgba 0.2891 0 1 1>}
  NamedRerouteDeclaration(None=@n30.Normal) {name=Root_Normal, nodeColor=<rgba 0 0.1797 1 1>}
  n61 = CollectionParameter() {
   collection=&FortniteMaterialParameters
   parameterName=SunlightDirectionalVector
  }
  n62 = ComponentMask(None=@n61) {b=true, g=true, r=true}
  n63 = Transform(None=NamedRerouteUsage())
  n64 = Multiply(A=DotProduct(A=@n62, B=@n63)) {constB=2}
  n65 = Normalize(VectorInput=Subtract(A=Multiply(A=@n64, B=@n63), B=@n62))
  n66 = Multiply(A=MaterialFunctionCall() {materialFunction=&CameraDirectionVector}.Result) {constB=-1}
  NamedRerouteDeclaration(None=ConstantBiasScale(None=DotProduct(A=@n65, B=@n66))) {name=LightVector, nodeColor=<rgba 0.3125 1 0 1>}
  n67 = Power(Base=NamedRerouteUsage(), Exp=ScalarParameter() {defaultValue=2, group=Tiger_Root, parameterName=TRoot_Blue_Fresnel_Power})
  n68 = ScalarParameter() {defaultValue=1, group=Tiger_Root, parameterName=TRoot_Blue_Fresnel_Contrast}
  n69 = Multiply(
   A=ComponentMask(None=NamedRerouteUsage()) {b=true}
   B=Saturate(None=MaterialFunctionCall(In=@n67, Contrast=@n68) {materialFunction=&CheapContrast}.Result)
  )
  NamedRerouteDeclaration(None=@n69) {name=MSk_Blue, nodeColor=<rgba 1 0 0.1875 1>}
  NamedRerouteDeclaration(None=Multiply(A=VectorParameter() {group=Tiger_Root, parameterName=TRoot_Blue_Color, defaultValue=<rgba 0.0369 0.8388 1 1>}.RGB, B=@n69)) {name=Color_Blue, nodeColor=<rgba 1 0 0.3281 1>}
  NamedRerouteDeclaration(None=@n30.EmissiveColor) {name=Root_Emissive, nodeColor=<rgba 0 0.1797 1 1>}
  n70 = NamedRerouteUsage()
  n71 = ScalarParameter() {defaultValue=1, group=Tiger_Root_Wave, parameterName=TRoot_Wave_A_Speed}
  n72 = Time()
  n73 = ScalarParameter() {defaultValue=1, group=Tiger_Root_Audio, parameterName=TRoot_Wave_A_Speed_Reactive}
  n74 = NamedRerouteUsage()
  n75 = Add(
   A=Multiply(B=ComponentMask(None=@n70) {r=true}) {constA=5}
   B=LinearInterpolate(A=Multiply(A=@n71, B=@n72), B=Multiply(A=@n72, B=@n73) {constB=1.1}, Alpha=Ceil(None=@n74))
  )
  n76 = Time()
  n77 = Sine(None=Multiply(A=Add(B=@n76) {constA=0.13}, B=ScalarParameter() {defaultValue=1.25, group=Tiger_Root_Wave, parameterName=TRoot_Wave_A_Hight_Speed}) {constB=0.5})
  n78 = Add(
   A=ScalarParameter() {defaultValue=0.12, group=Tiger_Root_Wave, parameterName=TRoot_Wave_A_height}
   B=Multiply(A=ConstantBiasScale(None=@n77)) {constB=0.05}
  )
  n79 = Multiply(A=ScalarParameter() {defaultValue=0.5, group=Tiger_Root_Audio, parameterName=TRoot_Wave_A_Audio_Strength}, B=NamedRerouteUsage())
  n80 = Ceil(None=NamedRerouteUsage())
  n81 = NamedRerouteUsage()
  n82 = ScalarParameter() {defaultValue=-0.655, group=Tiger_Root_Wave, parameterName=TRoot__Wave_Offset}
  n83 = Add(
   A=Multiply(A=Sine(None=@n75), B=LinearInterpolate(A=@n78, B=@n79, Alpha=@n80)) {constB=0.5}
   B=Multiply(A=Add(A=@n81, B=@n82) {constB=-0.7}) {constB=10}
  )
  n84 = ScalarParameter() {defaultValue=0.81, group=Tiger_Root_Wave, parameterName=TRoot_Wave_Skew_Amount}
  n85 = LinearInterpolate(A=ComponentMask(None=@n83) {g=true}, B=ComponentMask(None=@n83) {b=true}, Alpha=@n84)
  n86 = Saturate(None=Abs(None=Subtract(A=@n85) {constB=0.5}))
  n87 = ScalarParameter() {defaultValue=0.1, group=Tiger_Root_Wave, parameterName=TRoot_Wave_Core_Size, sortPriority=0}
  n88 = Saturate(None=OneMinus(None=Divide(A=@n86, B=@n87) {constB=0.1}))
  n89 = ScalarParameter() {defaultValue=5, group=Tiger_Root_Wave, parameterName=TRoot_Wave_Core_falloff, sortPriority=0}
  n90 = Power(Base=@n88, Exp=@n89) {constExponent=3}
  n91 = NamedRerouteUsage()
  n92 = ScalarParameter() {defaultValue=0.53, group=Tiger_Root_Wave, parameterName=TRoot_Wave_B_Speed}
  n93 = Time()
  n94 = ScalarParameter() {defaultValue=0.53, group=Tiger_Root_Audio, parameterName=TRoot_Wave_B_Speed_Reactive}
  n95 = NamedRerouteUsage()
  n96 = Add(
   A=Multiply(B=ComponentMask(None=@n91) {r=true}) {constA=4}
   B=LinearInterpolate(A=Multiply(A=@n92, B=@n93) {constB=0.5}, B=Multiply(A=@n93, B=@n94) {constB=0.75}, Alpha=Ceil(None=@n95))
  )
  n97 = Sine(None=Multiply(A=Time(), B=ScalarParameter() {defaultValue=1.25, group=Tiger_Root_Wave, parameterName=TRoot_Wave_B_Hight_Speed}) {constB=0.5})
  n98 = Add(
   A=ScalarParameter() {defaultValue=0.15, group=Tiger_Root_Wave, parameterName=TRoot_Wave_B_height}
   B=Multiply(A=ConstantBiasScale(None=@n97)) {constB=0.05}
  )
  n99 = Multiply(A=ScalarParameter() {defaultValue=0.5, group=Tiger_Root_Audio, parameterName=TRoot_Wave_B_Audio_Strength}, B=NamedRerouteUsage())
  n100 = Ceil(None=NamedRerouteUsage())
  n101 = NamedRerouteUsage()
  n102 = Add(
   A=Multiply(A=Sine(None=@n96), B=LinearInterpolate(A=@n98, B=@n99, Alpha=@n100)) {constB=0.5}
   B=Multiply(A=Add(A=@n101, B=@n82) {constB=-0.7}) {constB=10}
  )
  n103 = LinearInterpolate(A=ComponentMask(None=@n102) {g=true}, B=ComponentMask(None=@n102) {b=true}, Alpha=@n84)
  n104 = Saturate(None=Abs(None=Subtract(A=@n103) {constB=0.5}))
  n105 = Saturate(None=OneMinus(None=Divide(A=@n104, B=@n87) {constB=0.1}))
  n106 = Power(Base=@n105, Exp=@n89) {constExponent=3}
  n107 = Saturate(None=Add(A=@n90, B=@n106))
  n108 = ComponentMask(None=NamedRerouteUsage()) {r=true}
  n109 = Constant() {r=0.5}
  n110 = ComponentMask(None=NamedRerouteUsage()) {b=true, r=true}
  n111 = ComponentMask(None=VectorParameter() {group=Tiger_Root_Wave, parameterName=TRoot__Wave_Upper_Pos_Msk, defaultValue=<rgba 0.1 -0.3 0 1>}.RGB) {g=true, r=true}
  n112 = MaterialFunctionCall(Float2=MaterialFunctionCall(UVs=Add(A=@n110, B=@n111), Texture Scale=Constant2Vector() {g=0.3, r=0.22}) {materialFunction=&ScaleUVsByCenter}.UVs) {materialFunction=&BreakOutFloat2Components}
  n113 = Multiply(
   A=SphereMask(A=@n109, B=@n112.R, Hardness=Constant() {r=0.8}) {attenuationRadius=0.5}
   B=SphereMask(A=@n112.G, B=@n109) {attenuationRadius=0.5}
  )
  NamedRerouteDeclaration(None=Multiply(A=Multiply(A=@n107, B=@n108), B=@n113)) {name=Wave_Msk, nodeColor=<rgba 0.0312 1 0 1>}
  n114 = VectorParameter() {group=Tiger_Root_Wave, parameterName=TRoot_Wave_Ab_ColorAb, defaultValue=<rgba 1 0.6932 0.0543 1>}
  n115 = VectorParameter() {group=Tiger_Root_Wave, parameterName=TRoot_Wave_Aa_ColorAa, defaultValue=<rgba 5 4.3111 2.3849 1>}
  n116 = ScalarParameter() {defaultValue=2, group=Tiger_Root_Wave, parameterName=TRoot_Wave_Core_falloffB, sortPriority=0}
  n117 = VectorParameter() {group=Tiger_Root_Wave, parameterName=TRoot_Wave_Bb_ColorBb, defaultValue=<rgba 0 0.067 1 1>}
  n118 = VectorParameter() {group=Tiger_Root_Wave, parameterName=TRoot_Wave_Ba_ColorBa, defaultValue=<rgba 2.1039 2.4753 5 1>}
  n119 = Add(
   A=Multiply(A=LinearInterpolate(A=@n114.RGB, B=@n115.RGB, Alpha=@n90), B=Power(Base=@n88, Exp=@n116) {constExponent=3})
   B=Multiply(A=Power(Base=@n105, Exp=@n116) {constExponent=3}, B=LinearInterpolate(A=@n117.RGB, B=@n118.RGB, Alpha=@n106))
  )
  NamedRerouteDeclaration(None=Multiply(A=Multiply(A=@n119, B=@n108), B=@n113)) {name=Wave_Colors, nodeColor=<rgba 0.0312 1 0 1>}
  NamedRerouteDeclaration(None=Max(A=@n54, B=@n59)) {name=AudioSpectrum, nodeColor=<rgba 1 0.8203 0 1>}
  NamedRerouteDeclaration(None=MaterialFunctionCall(Float4=@n25.Spectrum) {materialFunction=&BreakOutFloat4Components}.G) {name=Audio_B, nodeColor=<rgba 0.2891 0 1 1>}
  NamedRerouteDeclaration(None=@n25.Amplitude) {name=Audio_C, nodeColor=<rgba 0.2891 0 1 1>}
  n120 = ComponentMask(None=NamedRerouteUsage()) {r=true}
  n121 = Time()
  n122 = Multiply(A=ScalarParameter() {defaultValue=1, group=Tiger_Root_Wave, parameterName=TRoot_Wave_C_Speed}, B=@n121)
  n123 = Multiply(A=@n121, B=ScalarParameter() {defaultValue=1, group=Tiger_Root_Audio, parameterName=TRoot_Wave_C_Speed_Reactive}) {constB=2.5}
  n124 = Ceil(None=NamedRerouteUsage())
  n125 = Sine(None=Add(A=Multiply(B=@n120) {constA=5}, B=LinearInterpolate(A=@n122, B=@n123, Alpha=@n124)))
  n126 = Add(B=Time()) {constA=0.13}
  n127 = ScalarParameter() {defaultValue=1.53, group=Tiger_Root_Wave, parameterName=TRoot_Wave_C_Hight_Speed}
  n128 = ConstantBiasScale(None=Sine(None=Multiply(A=@n126, B=@n127) {constB=0.5}))
  n129 = LinearInterpolate(
   A=Add(A=ScalarParameter() {defaultValue=0.6, group=Tiger_Root_Wave, parameterName=TRoot_Wave_C_height}, B=Multiply(A=@n128) {constB=0.1})
   B=Multiply(A=ScalarParameter() {defaultValue=1.5, group=Tiger_Root_Audio, parameterName=TRoot_Wave_C_Audio_Strength}, B=NamedRerouteUsage())
   Alpha=Ceil(None=NamedRerouteUsage())
  )
  n130 = ScalarParameter() {defaultValue=-0.17, group=Tiger_Root_Wave, parameterName=TRoot__Wave_Lower_Offset}
  n131 = Add(A=NamedRerouteUsage(), B=@n130) {constB=-0.7}
  n132 = ComponentMask(None=Add(A=Multiply(A=@n125, B=@n129) {constB=0.5}, B=Multiply(A=@n131) {constB=10})) {b=true}
  n133 = Saturate(None=Abs(None=Subtract(A=@n132) {constB=0.5}))
  n134 = ScalarParameter() {defaultValue=0.2, group=Tiger_Root_Wave, parameterName=TRoot_Wave_CoreB_Size, sortPriority=0}
  n135 = Saturate(None=OneMinus(None=Divide(A=@n133, B=@n134) {constB=0.1}))
  n136 = ScalarParameter() {defaultValue=5, group=Tiger_Root_Wave, parameterName=TRoot_Wave_CoreB_falloff, sortPriority=0}
  n137 = Power(Base=@n135, Exp=@n136) {constExponent=3}
  n138 = ComponentMask(None=NamedRerouteUsage()) {r=true}
  n139 = Time()
  n140 = Multiply(A=ScalarParameter() {defaultValue=0.53, group=Tiger_Root_Wave, parameterName=TRoot_Wave_D_Speed}, B=@n139) {constB=0.5}
  n141 = Multiply(A=@n139, B=ScalarParameter() {defaultValue=0.53, group=Tiger_Root_Audio, parameterName=TRoot_Wave_D_Speed_Reactive}) {constB=2}
  n142 = Ceil(None=NamedRerouteUsage())
  n143 = Sine(None=Add(A=Multiply(B=@n138) {constA=6}, B=LinearInterpolate(A=@n140, B=@n141, Alpha=@n142)))
  n144 = Time()
  n145 = ScalarParameter() {defaultValue=1.25, group=Tiger_Root_Wave, parameterName=TRoot_Wave_D_Hight_Speed}
  n146 = ConstantBiasScale(None=Sine(None=Multiply(A=@n144, B=@n145) {constB=0.5}))
  n147 = LinearInterpolate(
   A=Add(A=ScalarParameter() {defaultValue=0.75, group=Tiger_Root_Wave, parameterName=TRoot_Wave_D_height}, B=Multiply(A=@n146) {constB=0.1})
   B=Multiply(A=ScalarParameter() {defaultValue=1.5, group=Tiger_Root_Audio, parameterName=TRoot_Wave_D_Audio_Strength}, B=NamedRerouteUsage())
   Alpha=Ceil(None=NamedRerouteUsage())
  )
  n148 = Add(A=NamedRerouteUsage(), B=@n130) {constB=-0.7}
  n149 = ComponentMask(None=Add(A=Multiply(A=@n143, B=@n147) {constB=0.5}, B=Multiply(A=@n148) {constB=10})) {b=true}
  n150 = Saturate(None=Abs(None=Subtract(A=@n149) {constB=0.5}))
  n151 = Saturate(None=OneMinus(None=Divide(A=@n150, B=@n134) {constB=0.1}))
  n152 = Power(Base=@n151, Exp=@n136) {constExponent=3}
  n153 = Add(A=@n137, B=@n152)
  n154 = ComponentMask(None=NamedRerouteUsage()) {r=true}
  NamedRerouteDeclaration(None=Multiply(A=Saturate(None=@n153), B=@n154)) {name="Wave_Msk_C\\D", nodeColor=<rgba 0.0312 1 0 1>}
  n155 = ScalarParameter() {defaultValue=3, group=Tiger_Root_Wave, parameterName=TRoot_Wave_CoreB_falloffB, sortPriority=0}
  n156 = Multiply(
   A=LinearInterpolate(
    A=VectorParameter() {group=Tiger_Root_Wave, parameterName=TRoot_Wave_Cb_Color, defaultValue=<rgba 0.9131 0.0107 0.8752 1>}.RGB
    B=VectorParameter() {group=Tiger_Root_Wave, parameterName=TRoot_Wave_Ca_Color, defaultValue=<rgba 5 0.9062 4.5412 1>}.RGB
    Alpha=Power(Base=@n135, Exp=@n155) {constExponent=3}
   )
   B=@n137
  )
  n157 = Multiply(
   A=Power(Base=@n151, Exp=@n155) {constExponent=3}
   B=LinearInterpolate(
    A=VectorParameter() {group=Tiger_Root_Wave, parameterName=TRoot_Wave_Db_Color, defaultValue=<rgba 1 0.6368 0.027 1>}.RGB
    B=VectorParameter() {group=Tiger_Root_Wave, parameterName=TRoot_Wave_Da_Color, defaultValue=<rgba 5 4.3159 0.7292 1>}.RGB
    Alpha=@n152
   )
  )
  NamedRerouteDeclaration(None=Multiply(A=Add(A=@n156, B=@n157), B=@n154)) {name="Wave_Colors_C\\D", nodeColor=<rgba 0.0312 1 0 1>}
  n158 = Power(
   Base=ComponentMask(None=NamedRerouteUsage()) {r=true}
   Exp=ScalarParameter() {defaultValue=1, group=Tiger_Root, parameterName=TRoot_RedHair_Fresnel_Power}
  )
  n159 = ScalarParameter() {group=Tiger_Root, parameterName=TRoot_RedHair_Fresnel_Contrast}
  n160 = Saturate(None=Add(
    A=NamedRerouteUsage()
    B=MaterialFunctionCall(In=@n158, Contrast=@n159) {materialFunction=&CheapContrast}.Result
   ))
  NamedRerouteDeclaration(None=@n160) {name=MSk_Hair, nodeColor=<rgba 1 0 0.1875 1>}
  n161 = Add(A=NamedRerouteUsage(), B=NamedRerouteUsage())
  n162 = Power(Base=OneMinus(None=Saturate(None=@n161)))
  n163 = NamedRerouteUsage()
  n164 = CurveAtlasRowParameter(CurveTime=NamedRerouteUsage()) {
   atlas=/BRCosmetics/Characters/Player/Male/Medium/Bodies/M_MED_Cactus_Dancer/Skin/M_MED_TigerRootHype_Blue/Textures/CA_TigerRootH
   curve=/BRCosmetics/Characters/Player/Male/Medium/Bodies/M_MED_Cactus_Dancer/Skin/M_MED_TigerRootHype_Blue/Textures/CC_TigerRoot_Shirt
   group=Tiger_Root
   parameterName=TRoot_ColorC_Msk_Red_Clothes
  }
  n165 = NamedRerouteUsage()
  n166 = CurveAtlasRowParameter(CurveTime=NamedRerouteUsage()) {
   atlas=&CA_TigerRootH
   curve=/BRCosmetics/Characters/Player/Male/Medium/Bodies/M_MED_Cactus_Dancer/Skin/M_MED_TigerRootHype_Blue/Textures/CC_TigerRoot_Green_orange_Skin
   defaultValue=2
   group=Tiger_Root
   parameterName=TRoot_ColorC_Msk_BlueSkin
  }
  n167 = Multiply(A=OneMinus(None=NamedRerouteUsage()), B=@n166)
  n168 = ScalarParameter() {group=Tiger_Root, parameterName="TRoot-Skin--BaseColor--->Emissve"}
  n169 = ComponentMask(None=NamedRerouteUsage()) {b=true}
  n170 = LinearInterpolate(
   A=LinearInterpolate(A=@n163, B=Multiply(A=@n163, B=@n164.R), Alpha=ComponentMask(None=@n165) {r=true})
   B=LinearInterpolate(A=@n167, Alpha=NamedRerouteDeclaration(None=@n168) {name=SkinTransition, nodeColor=<rgba 0.0312 1 0 1>}) {constB=0}
   Alpha=@n169
  ) {constB=0}
  n171 = Add(A=Add(A=NamedRerouteUsage(), B=NamedRerouteUsage()), B=NamedRerouteUsage())
  n172 = NamedRerouteUsage()
  n173 = OneMinus(None=Saturate(None=Add(A=@n171, B=@n172)))
  n174 = Multiply(A=@n162, B=Multiply(A=NamedRerouteDeclaration(None=@n170) {name=BC_SkinBlue, nodeColor=<rgba 0.6875 1 0 1>}, B=@n173))
  n175 = CurveAtlasRowParameter(CurveTime=NamedRerouteUsage()) {
   atlas=&CA_TigerRootH
   curve=/BRCosmetics/Characters/Player/Male/Medium/Bodies/M_MED_Cactus_Dancer/Skin/M_MED_TigerRootHype_Blue/Textures/CC_TigerRoot_Green_Straps
   defaultValue=5
   group=Tiger_Root
   parameterName=TRoot_ColorC_Msk_Green_Straps
  }
  n176 = Multiply(A=@n175, B=NamedRerouteUsage())
  n177 = NamedRerouteUsage()
  NamedRerouteDeclaration(None=Add(A=LinearInterpolate(A=@n174, B=@n176, Alpha=@n177) {constB=0}, B=LinearInterpolate(B=@n34, Alpha=@n35)) {constB=0}) {name=BaseColor, nodeColor=<rgba 1 0.9375 0 1>}
  n178 = Power(Base=NamedRerouteUsage(), Exp=ScalarParameter() {defaultValue=1.5, group=Tiger_Root, parameterName=TRoot_Red_Fresnel_Power})
  n179 = ScalarParameter() {defaultValue=1, group=Tiger_Root, parameterName=TRoot_Red_Fresnel_Contrast}
  n180 = Multiply(
   A=ComponentMask(None=NamedRerouteUsage()) {r=true}
   B=Saturate(None=MaterialFunctionCall(In=@n178, Contrast=@n179) {materialFunction=&CheapContrast}.Result)
  )
  NamedRerouteDeclaration(None=@n180) {name=MSk_Red_A, nodeColor=<rgba 1 0 0.1875 1>}
  n181 = NamedRerouteDeclaration(None=Multiply(A=@n167, B=@n169)) {name=Skin_Msked, nodeColor=<rgba 1 0.7734 0 1>}
  NamedRerouteDeclaration(None=Multiply(A=@n162, B=Multiply(A=@n181, B=@n173))) {name=Skin_B, nodeColor=<rgba 0.2422 1 0 1>}
  NamedRerouteDeclaration(None=@n39) {name=MSk_GreenUpper_B, nodeColor=<rgba 1 0 0.1875 1>}
  n182 = NamedRerouteUsage()
  n183 = Power(Base=NamedRerouteUsage(), Exp=ScalarParameter() {defaultValue=1.5, group=Tiger_Root, parameterName=TRoot_GreenLower_Fresnel_Power})
  n184 = ScalarParameter() {defaultValue=1, group=Tiger_Root, parameterName=TRoot_GreenLower_Fresnel_Contrast}
  n185 = Multiply(
   A=Multiply(A=ComponentMask(None=@n182) {g=true}, B=NamedRerouteUsage())
   B=Saturate(None=MaterialFunctionCall(In=@n183, Contrast=@n184) {materialFunction=&CheapContrast}.Result)
  )
  NamedRerouteDeclaration(None=Multiply(A=VectorParameter() {group=Tiger_Root, parameterName=TRoot_GreenLower_Color, defaultValue=<rgba 0.1587 0.0206 0.3958 1>}.RGB, B=@n185)) {name=Color_Green_Lower, nodeColor=<rgba 0.4531 1 0 1>}
  NamedRerouteDeclaration(None=Multiply(A=VectorParameter() {group=Tiger_Root, parameterName=TRoot_GreenUpper_Color, defaultValue=<rgba 0 0.1647 0.25 1>}.RGB, B=@n42)) {name=Color_Green_Upper, nodeColor=<rgba 0.4531 1 0 1>}
  NamedRerouteDeclaration(None=Multiply(A=VectorParameter() {group=Tiger_Root, parameterName=TRoot_Red_Color, defaultValue=<rgba 0.9911 0.2346 0.9647 1>}.RGB, B=@n180)) {name=Color_Red, nodeColor=<rgba 1 0 0.3281 1>}
  NamedRerouteDeclaration(None=Multiply(A=VectorParameter() {group=Tiger_Root, parameterName=TRoot_Red_Hair, defaultValue=<rgba 0.9911 0.2346 0.9647 1>}.RGB, B=@n160)) {name="Color Red\\Hair", nodeColor=<rgba 0 0.2266 1 1>}
  NamedRerouteDeclaration(None=Multiply(A=VectorParameter() {group=Tiger_Root, parameterName=TRoot_Green_Color, defaultValue=<rgba 0.0176 0.0176 0.0176 1>}.RGB, B=@n48)) {name=Color Green Smpl, nodeColor=<rgba 0 0.2266 1 1>}
  NamedRerouteDeclaration(None=Multiply(A=VectorParameter() {group=Tiger_Root, parameterName=TRoot_Red_Color, defaultValue=<rgba 0.9911 0.2346 0.9647 1>}.RGB, B=@n51)) {name=Color Red Smpl, nodeColor=<rgba 0 0.2266 1 1>}
  n186 = TextureSampleParameter2D() {
   group=Tiger_Root
   parameterName=TRoot__Tex_FX_A
   samplerType=SAMPLERTYPE_Masks
   texture=/BRCosmetics/Characters/Player/Female/Medium/Bodies/F_MED_Cactus_Rocker/Skins/TigerRootFame_Orange/Textures/T_F_MED_TigerRootFame_Orange_Body_FX
  }
  NamedRerouteDeclaration(None=@n186.RGBA) {name=Msk_FX_A, nodeColor=<rgba 0.7578 0 1 1>}
  NamedRerouteDeclaration(None=@n35) {name=Emissive--BaseColor, nodeColor=<rgba 1 0.4922 0 1>}
  n187 = NamedRerouteUsage()
  n188 = Multiply(A=@n187, B=ScalarParameter() {defaultValue=0.5, group=Tiger_Root, parameterName=TRoot_Fresnel_audio Strength})
  NamedRerouteDeclaration(None=LinearInterpolate(B=Add(A=@n188) {constB=0.8}, Alpha=Ceil(None=@n187)) {constA=1}) {name=Audio_reactive_a, nodeColor=<rgba 1 0 0.1172 1>}
  NamedRerouteDeclaration(None=@n60) {name=Audio_A, nodeColor=<rgba 0.2891 0 1 1>}
  NamedRerouteDeclaration(None=@n30.Specular) {name=Root_Spec, nodeColor=<rgba 0 0.1797 1 1>}
  n189 = Transform(None=@n27.RGB)
  NamedRerouteDeclaration(None=FunctionInput(Preview=Fresnel(Normal=@n189) {baseReflectFraction=0, exponent=1}) {bUsePreviewValueAsDefault=true, inputName=Fresnel, inputType=FunctionInput_Scalar}) {name=Msk_Fresenl, nodeColor=<rgba 1 0 0.1172 1>}
  NamedRerouteDeclaration(None=@n185) {name=MSk_RedLower, nodeColor=<rgba 1 0 0.1875 1>}
  n190 = FunctionInput(Preview=VertexInterpolator(VS=PreSkinnedPosition()).PS) {inputName=Pre_Skin_Pos}
  n191 = PreSkinnedLocalBounds()
  n192 = MaterialFunctionCall(Float3=Subtract(A=@n191.Max, B=@n191.Max)) {materialFunction=&BreakOutFloat3Components}
  n193 = Max(A=@n192.R, B=@n192.R)
  NamedRerouteDeclaration(None=Divide(A=Subtract(A=@n190, B=@n191.Min), B=Max(A=@n193, B=@n192.B))) {name=Bounds, nodeColor=<rgba 0 0.0859 1 1>}
  n194 = ComponentMask(None=@n31) {b=true}
  n195 = ScalarParameter() {defaultValue=-0.4, group=Tiger_Root, parameterName=TRoot_Masking_Green_Upper}
  NamedRerouteDeclaration(None=Ceil(None=Add(A=@n194, B=@n195))) {name=Msk_Upper, nodeColor=<rgba 0 0.3437 1 1>}
  NamedRerouteDeclaration(None=VertexInterpolator(VS=PreSkinnedNormal()).PS) {name=PreSkin_Norm, nodeColor=<rgba 0 0.1797 1 1>}
  NamedRerouteUsage()
  NamedRerouteUsage()
  NamedRerouteUsage()
  NamedRerouteUsage()
  NamedRerouteUsage()
  PixelNormalWS()
  Saturate()
  n196 = Abs(None=ComponentMask(None=NamedRerouteUsage()) {g=true})
  n197 = ScalarParameter() {defaultValue=3, group=Tiger_Root_Wave, parameterName="TRoot_Wave_C\\D_Msk_Power"}
  Saturate(None=MaterialFunctionCall(
    In=Power(Base=@n196, Exp=@n197)
    Contrast=ScalarParameter() {defaultValue=0.05, group=Tiger_Root_Wave, parameterName="TRoot_Wave_C\\D_Msk_Contrast"}
   ) {materialFunction=&CheapContrast}.Result)
  n198 = NamedRerouteUsage()
  n199 = NamedRerouteUsage()
  SetMaterialAttributes!(
   MaterialAttributes=NamedRerouteUsage()
   BaseColor=LinearInterpolate(A=@n198, Alpha=@n199) {constB=0}
   EmissiveColor=LinearInterpolate(B=@n198, Alpha=@n199)
   Normal
   Specular
  ) {attributeSetTypes [
    #1
    #2
    #3
    #4
   ]}
  SmoothStep()
  SmoothStep()
  VertexNormalWS()
 ]}
