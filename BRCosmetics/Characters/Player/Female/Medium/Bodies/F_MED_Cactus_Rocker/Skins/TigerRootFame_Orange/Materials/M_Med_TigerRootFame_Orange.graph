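# Skeletal mesh material for the orange TigerRootFame skin of the female medium Cactus Rocker body, using material attributes and Default Lit shading.
# The shared PreFX character function supplies inputs such as pre-skinned local position to MF_TigerRootFame, which builds the skin-specific look.
# That result feeds the shared PostFX character function alongside PreFX outputs.
# A separate CharacterSkinFX call also reads from PreFX.
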Material /BRCosmetics/Characters/Player/Female/Medium/Bodies/F_MED_Cactus_Rocker/Skins/TigerRootFame_Orange/Materials/M_Med_TigerRootFame_Orange
graph {
 properties {
  bUseMaterialAttributes=true
  bUsedWithMeshDeformer=true
  bUsedWithMorphTargets=true
  bUsedWithSkeletalMesh=true
  pixelDepthOffsetMode=PDOM_Legacy
  refractionMethod=RM_IndexOfRefraction
  usedShadingModels=Default Lit
 }
 nodes [
  n1 = MaterialFunctionCall() {materialFunction=&BaseFN_CharacterFunction_PreFX}
  n2 = MaterialFunctionCall!(Pre_Skin_Pos=@n1.Pre skinned Local Position, Fresnel=@n1.Pre skinned Local Position, Mat=@n1.Pre skinned Local Position) {materialFunction=&/MF_TigerRootFame}
  MaterialFunctionCall(AO=@n1.AO, Distance Blend=@n1.AO, Fresnel=@n1.AO, Gmap=@n1.AO, Material=@n2.Result) {materialFunction=&BaseFN_CharacterFunction_PostFX}
  MaterialFunctionCall(
   PreskinnedPosition=@n1.Pre skinned Local Position
   Gmap=@n1.Pre skinned Local Position
   MaterialAttributes=@n1.Pre skinned Local Position
   AO=@n1.Pre skinned Local Position
  ) {materialFunction=&CharacterSkinFX}
 ]
}
